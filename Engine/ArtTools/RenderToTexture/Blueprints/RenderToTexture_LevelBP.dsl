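; Render-to-texture art tool Blueprint that lays out meshes for capturing imposter sprites, flipbooks and depth maps, with optional physics-drop placement.
; Render type and imposter mode drive a grid of rotated, scaled mesh copies with per-frame material instances and a preview material.
; Motion vector passes add a second UV layout sheet, faking dilation with stepped, offset mesh copies.
; Imposter transforms are stored, and bounds radius and offset are rendered as text.
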(event EventBeginPlay
  (bind _simulate_physics (Variables|PhysicsGround|GetSimulate%20Physics))
  ;[1 1 1] Physics Drop
  (Utilities|FlowControl|Sequence
    (:then_0
      (switch Utilities|FlowControl|Switch|SwitchonRenderToTexture_Enum (Variables|CaptureSettings|GetRender%20Type)
        (:NewEnumerator0
          _)
        (:NewEnumerator6
          _)
        (:NewEnumerator8
          (if (or _simulate_physics (Variables|PhysicsGround|GetSimple%20Random%20Placement))
            (Utilities|PhysDropEventGraph _simulate_physics))))
      (Rendering|Material|SetScalarParameterValue "/Engine/ArtTools/RenderToTexture/Materials/MPC_R2T.MPC_R2T" "RenderMotionVectorMultiply" (Flipbook/ImposterCombiners|Flipbook-ImposterParamChooser (Variables|Render3DImposterSprites|GetMotion%20Vector%20IntensityBoost) (Variables|RenderFlipbook|GetMotionVectorIntensityBoost)) :!class /Script/Engine.KismetMaterialLibrary))
    (:then_1
      (Rendering|Material|SetScalarParameterValue 0 "sRGB" (Math|Conversions|ToFloat(Boolean) (Variables|Default|GetsRGB)) :!class /Script/Engine.KismetMaterialLibrary))))

(event Custom|Match%20Level%20Actors%20to%20Imposter%20Array
  ;[1 1 1] A hack to modify level actors to match the imposter array for specific Subsurface Lightmass baking tests
  (bind _outactors (Actor|GetAllActorsOfClass "/Script/Engine.StaticMeshActor"))
  (for _index (range (+ (Utilities|Array|Length _outactors) 1))
    (if (Utilities|Equal(Object) (Variables|/Script/Engine.StaticMeshComponent|GetStaticMesh (Variables|/Script/Engine.StaticMeshActor|GetStaticMeshComponent (Utilities|Array|Get(acopy) _outactors _index))) (Variables|Render3DImposterSprites|GetImposter%20Static%20Mesh))
      (Utilities|Array|Add (Variables|Default|GetLevelStaticMeshActorArray) (Utilities|Array|Get(acopy) _outactors _index))))
  (for (_array_element _array_index) (Variables|Default|GetLevelStaticMeshActorArray)
    (Transformation|SetActorTransform _array_element (Utilities|Array|Get(acopy) (Variables|Render3DImposterSprites|GetImposterArray) _array_index) false false)))

(fn Copyamesh
  (params (Mesh%20Component :type Static%20Mesh%20Component%20Object%20Reference) (Offset%20Vector :type Vector) (Array%20to%20Place%20Meshes%20in :type Array%20of%20Static%20Mesh%20Component%20Object%20References))
  (if (Utilities|NotEqual(Object) Mesh%20Component 0)
    (bind _returnvalue (AddComponent|NODE_AddStaticMeshComponent-5 :bManualAttachment true :RelativeTransform (Math|Transform|MakeTransform (- (.location (Transformation|GetWorldTransform Mesh%20Component)) (* (Math|Transform|TransformDirection (Transformation|GetActorTransform) Offset%20Vector) (Variables|Debug|GetSheetSize))) (.rotation (Transformation|GetWorldTransform Mesh%20Component)) (.scale (Transformation|GetWorldTransform Mesh%20Component)))))
    (Components|StaticMesh|SetStaticMesh _returnvalue (Variables|/Script/Engine.StaticMeshComponent|GetStaticMesh Mesh%20Component))
    (Utilities|Array|Add Array%20to%20Place%20Meshes%20in _returnvalue)
    ;[1 1 1] Copy all Material IDs
    (for _index (range (+ (Rendering|Material|GetNumMaterials Mesh%20Component) 1))
      (Rendering|Material|SetMaterial _returnvalue _index (Rendering|Material|GetMaterial Mesh%20Component _index)))))

(fn FindZBounds
  (params (Meshes :type Array%20of%20Static%20Mesh%20Component%20Object%20References))
  (LocalVariables|SetLocalMin 1000000000000)
  (LocalVariables|SetLocalMax -100000000000000)
  (for _array_element Meshes
    (if (< (- (.z (Collision|GetComponentBounds _array_element)) (.z (.BoxExtent (Collision|GetComponentBounds _array_element)))) (LocalVariables|GetLocalMin))
      (LocalVariables|SetLocalMin (- (.z (Collision|GetComponentBounds _array_element)) (.z (.BoxExtent (Collision|GetComponentBounds _array_element)))))
      (elif (> (+ (.z (Collision|GetComponentBounds _array_element)) (.z (.BoxExtent (Collision|GetComponentBounds _array_element)))) (LocalVariables|GetLocalMax))
        (LocalVariables|SetLocalMax (+ (.z (Collision|GetComponentBounds _array_element)) (.z (.BoxExtent (Collision|GetComponentBounds _array_element))))))))
  (return (LocalVariables|GetLocalMin) (LocalVariables|GetLocalMax)))

(fn FlipBook
  (bind _offst_phase (Variables|Default|GetOffst%20Phase))
  (bind _returnvalue (* _offst_phase 360))
  (bind _returnvalue_2 (Flipbook/ImposterCombiners|UVDilationStepParamChooser))
  (bind (_x _y) (Flipbook/ImposterCombiners|FrameCountParamChooser))
  (bind (_x_2 _y_2) (Flipbook/ImposterCombiners|FrameCountParamChooser))
  (bind (_x_3 _y_3) (Flipbook/ImposterCombiners|FrameCountParamChooser))
  (bind _flipbookmid (Variables|RenderFlipbook|GetFlipbookMID))
  (bind _flipbookmid_2 (Variables|RenderFlipbook|GetFlipbookMID))
  (bind _flipbookmid_3 (Variables|RenderFlipbook|GetFlipbookMID))
  (bind _flipbookmid_4 (Variables|RenderFlipbook|GetFlipbookMID))
  (bind _flipbookmid_5 (Variables|RenderFlipbook|GetFlipbookMID))
  (bind _staticmeshcomponent (Variables|Default|GetStaticMeshComponent))
  (bind _staticmeshcomponent_2 (Variables|Default|GetStaticMeshComponent))
  (bind _internalscalefactor (Variables|Default|GetInternalscalefactor))
  (bind _dilation_phase (Variables|Default|GetDilation%20Phase))
  (bind _staticmeshcomponent_3 (Variables|Default|GetStaticMeshComponent))
  (bind _flipbookmid_6 (Variables|RenderFlipbook|GetFlipbookMID))
  (bind _staticmeshcomponent_4 (Variables|Default|GetStaticMeshComponent))
  (Utilities|Array|Clear (Variables|RenderFlipbook|GetFlipbookMeshes))
  (Utilities|FlowControl|Sequence
    (:then_0
      (Variables|Default|SetCurrentRenderLoc (Transformation|GetActorLocation))
      (Variables|Default|SetSectorSize (/ (Variables|Debug|GetSheetSize) (Flipbook/ImposterCombiners|Flipbook-ImposterParamChooser (Math|Conversions|ToFloat(Integer) (Variables|Render3DImposterSprites|GetFrames%20around%20Z%20rotation)) (Math|Conversions|ToFloat(Integer) (Math|Integer|Max(Integer) (Variables|RenderFlipbook|GetFlipbook%20Columns%20%28X%29) (Variables|RenderFlipbook|GetFlipbook%20Rows%20%28Y%29))))))
      (if (Utilities|Flipbook-ImposterMotionVectorChooser)
        (Utilities|FlowControl|Sequence
          (:then_0
            ;[0.345 0.345 0.345] Applied Motion Vector Setup
            ;; MotionVector Background Setup
            (Transformation|AddLocalOffset (Variables|Default|GetPlaneMesh) (* "0, 0,-1.000000" (* (Math|Conversions|ToFloat(Integer) (Flipbook/ImposterCombiners|MotionDilationStepParamChooser)) (Variables|Default|GetSectorSize))) false false)
            (Rendering|Material|SetMaterial (Variables|Default|GetPlaneMesh) 0 (select (Variables|RenderFlipbook|GetRender%20Motion%20Vectors) "/Engine/ArtTools/RenderToTexture/Materials/Debug/M_Emissive_Color_127-127-0.M_Emissive_Color_127-127-0" (Variables|Default|GetBackgroundSheetMID))))
          (:then_1
            ;[1 0.8243 0] UV WorldPosition Layout Setup
            ;; Motion Vector UV Layout Setup
            (Variables|Default|SetMotionVector%20Pass true)
            (Variables|Default|SetUV%20DilationPass true)
            (Variables|Default|SetCurrentRenderLoc (+ (Transformation|GetActorLocation) (Math|Vector|MakeVector 0.0 (Math|Conversions|ToFloat(Integer) (Math|Float|Floor (* (Variables|Debug|GetSheetSize) -1.25))))))
            (CallFunction|NewMesh "/Engine/ArtTools/RenderToTexture/Meshes/S_1_Unit_Plane.S_1_Unit_Plane" (+ (Variables|Default|GetCurrentRenderLoc) (+ (Math|Vector|MakeVector 0.0 0.0 (Variables|CaptureSettings|GetSheet%20height)) (* "0, 0,-1.000000" (* (Math|Conversions|ToFloat(Integer) (Flipbook/ImposterCombiners|UVDilationStepParamChooser)) (* (Variables|Debug|GetSheetSize) 0.1))))) "/Engine/ArtTools/RenderToTexture/Materials/Debug/M_Emissive_Color_127-127-0.M_Emissive_Color_127-127-0" (Math|Conversions|ToVector(Float) (* (Variables|Debug|GetSheetSize) 1)))))
        (else
          (Variables|RenderFlipbook|SetMaterialChoice 0)))
      ;[0.3656 1 0.2591] 2D Grid Setup
      (bind (_current_point _total_points _0_1_phase _poly_center _x_row _y_row) (Utilities|2DGridExecutionMacro (Variables|Default|GetSectorSize) (Variables|Default|GetCurrentRenderLoc) _x_3 _y_3)
        (:Per%20Poly%20Execution
          (bind (_offset_phase _new_center _dilation_phase_2) (Utilities|MotionVectorSubstepDilation _total_points (Flipbook/ImposterCombiners|MotionDilationStepParamChooser) _0_1_phase (Variables|Default|GetSectorSize) _poly_center)
            (:Per%20Step%20Dilation%20Execution
              (Variables|Default|SetOffst%20Phase _offset_phase)
              (Variables|Default|SetDilation%20Phase _dilation_phase_2)
              (Variables|Default|SetPolyCenter (Utilities|StadiumOffsetting _new_center _y_row))
              ;[0.1613 0.4521 1] Add Meshes
              (bind _returnvalue_3 (AddComponent|NODE_AddStaticMeshComponent-0 :bManualAttachment true :RelativeTransform (Math|Conversions|ToTransform(Vector) (Utilities|StadiumOffsetting _new_center _y_row))))
              (Variables|Default|SetStaticMeshComponent _returnvalue_3)
              (if (Variables|Default|GetUV%20DilationPass)
                (Utilities|Array|Add (Variables|RenderFlipbook|GetUnwrappedMotionVectorMeshes) _staticmeshcomponent_3)
                (else
                  (Utilities|Array|Add (Variables|RenderFlipbook|GetFlipbookMeshes) _staticmeshcomponent_3)))
              (Components|StaticMesh|SetStaticMesh (Variables|Default|GetStaticMeshComponent) (Flipbook/ImposterCombiners|Flipbook-ImposterMeshChooser))
              ;[1 1 1] Size Calculation
              (Variables|Default|SetInitialOffset (- (Variables|Default|GetPolyCenter) (Collision|GetComponentBounds _returnvalue_3)))
              (Variables|Default|SetInternalscalefactor (* (/ (Variables|Default|GetSectorSize) (* (.SphereRadius (Collision|GetComponentBounds _returnvalue_3)) 2)) (Flipbook/ImposterCombiners|Flipbook-ImposterParamChooser (Variables|Render3DImposterSprites|GetMeshScale) (Variables|RenderFlipbook|GetFlipbook%20Mesh%20scale))))
              (switch Utilities|FlowControl|Switch|SwitchonRenderToTexture_Enum (Variables|CaptureSettings|GetRender%20Type)
                (:NewEnumerator0
                  (switch Utilities|FlowControl|Switch|SwitchonRenderToTexture_Imposter_Enum (Variables|Render3DImposterSprites|GetImposterEnum)
                    (:NewEnumerator0
                      ;[0.195 0.195 0.195] Imposter
                      ;;[1 1 1] 3D Rotation
                      (Transformation|AddLocalRotation (Variables|Default|GetStaticMeshComponent) (Math|Rotator|MakeRotator (* (/ (+ _dilation_phase (Math|Conversions|ToFloat(Integer) _y_row)) (- (Math|Conversions|ToFloat(Integer) _y) 1.0)) -180)) false false)
                      (Transformation|AddLocalRotation (Variables|Default|GetStaticMeshComponent) (Math|Rotator|MakeRotator 0.0 0.0 (* (/ (+ _dilation_phase (Math|Conversions|ToFloat(Integer) _x_row)) (Math|Conversions|ToFloat(Integer) _x)) 360)) false false))
                    (:NewEnumerator1
                      ;;[1 1 1] Single Rotation Axis
                      (Transformation|AddLocalRotation (Variables|Default|GetStaticMeshComponent) (Math|Rotator|MakeRotfromZ (Variables|Render3DImposterSprites|GetSingle%20Rotation%20Axis)) false false)
                      (Transformation|AddLocalRotation (Variables|Default|GetStaticMeshComponent) (Math|Rotator|MakeRotator 0.0 0.0 (* (Variables|Default|GetOffst%20Phase) 360)) false false)))
                  ;;[1 1 1] Offsetting
                  (Transformation|AddLocalOffset _staticmeshcomponent_2 (* (Variables|Default|GetInitialOffset) _internalscalefactor) false false)
                  (Transformation|SetRelativeScale3D _staticmeshcomponent_2 (Math|Conversions|ToVector(Float) _internalscalefactor))
                  ;;[1 1 1] Motion Vector Option
                  (if (Utilities|Flipbook-ImposterMotionVectorChooser)
                    ;;[1 1 1] Motion Vector passes
                    (Variables|RenderFlipbook|SetMaterialChoice (select (Variables|Default|GetUV%20DilationPass) (Variables|Default|GetMotionVectorMaterial) (Variables|Default|GetMotionVectorsApplied)))
                    (if (Variables|Default|GetUV%20DilationPass)
                      (Utilities|FlowControl|Sequence
                        (:then_0
                          ;;[0.0787 0.5836 1] MID setup for Motion Vector Capture
                          (bind _returnvalue_4 (Rendering|Material|CreateDynamicMaterialInstance (Variables|RenderFlipbook|GetMaterialChoice) "None" "None" :!class /Script/Engine.KismetMaterialLibrary))
                          (Variables|RenderFlipbook|SetFlipbookMID _returnvalue_4))
                        (:then_1
                          ;;[0.4937 0.735 0] Motion Vector UV Layout
                          (Utilities|ApplyMIDToAllMeshMatIDs (Variables|Default|GetStaticMeshComponent) (Variables|RenderFlipbook|GetFlipbookMID))))
                      (else
                        (Utilities|FlowControl|Sequence
                          (:then_0
                            ;;[1 1 1] Motion Vector Masked Setup
                            (bind _mid (Utilities|CreateAndAssignMIDsFromMaskArrays (Utilities|Select :Index (Variables|Render3DImposterSprites|GetImposterEnum) :NewEnumerator0 (Variables|Default|GetMotionVectorsApplied_Imposter) :NewEnumerator1 (Variables|Default|GetMotionVectorsApplied)) (Variables|Default|GetStaticMeshComponent) (Variables|Render3DImposterSprites|GetOpacity%20Mask%20Textures) (Variables|Render3DImposterSprites|GetOpacity%20Mask%20Channels))
                              (:Loop%20Body%20Continue
                                (Variables|RenderFlipbook|SetFlipbookMID _mid))))
                          (:then_1
                            (Utilities|FlowControl|Sequence
                              (:then_0
                                ;;[1 1 1] Store array of imposter info
                                (Utilities|Array|Add (Variables|Render3DImposterSprites|GetImposterArray) (Transformation|GetWorldTransform (Variables|Default|GetStaticMeshComponent)))
                                ;[1 1 1] Imposter Depth Mask Setup
                                (if (Variables|Render3DImposterSprites|GetDebug%20Depth)
                                  (Utilities|CreateAndAssignMIDsFromMaskArrays "/Engine/ArtTools/RenderToTexture/Materials/M_RenderNormalizedDepth_01_Inst.M_RenderNormalizedDepth_01_Inst" (Variables|Default|GetStaticMeshComponent) (Variables|Render3DImposterSprites|GetOpacity%20Mask%20Textures) (Variables|Render3DImposterSprites|GetOpacity%20Mask%20Channels))))
                              (:then_1
                                ;;[0.2707 0.1032 0.34] Size and offset info to use in final material
                                (bind _returnvalue_5 (AddComponent|NODE_AddTextRenderComponent-0 :bManualAttachment true :RelativeTransform (Math|Transform|MakeTransform (* (Math|Vector|MakeVector (Variables|Debug|GetSheetSize)) 0.625) "90.000000,180.000000,90.000000" (Math|Conversions|ToVector(Float) (Variables|RenderDepthMap|GetTextSize)))))
                                (Rendering|Components|TextRender|SetText _returnvalue_5 (Utilities|Text|ToText(String) (Utilities|String|BuildString(Float) "Boundsphere Radius: " "" (Variables|Default|Getinitialscale) " ")))
                                (bind _returnvalue_6 (AddComponent|NODE_AddTextRenderComponent-1 :bManualAttachment true :RelativeTransform (Math|Transform|MakeTransform (+ (* (Math|Vector|MakeVector (Variables|Debug|GetSheetSize)) 0.625) "0,-25.000000, 0") "90.000000,180.000000,90.000000" (Math|Conversions|ToVector(Float) (Variables|RenderDepthMap|GetTextSize)))))
                                (Rendering|Components|TextRender|SetText _returnvalue_6 (Utilities|Text|ToText(String) (Utilities|String|Append "Offset Vector: " (Utilities|String|ToString(Vector) (Variables|Default|GetInitialOffset)))))))))))
                    (Rendering|Material|SetVectorParameterValue _flipbookmid_6 "FrameCenter" (Math|Conversions|ToLinearColor(Vector) (Variables|Default|GetPolyCenter)) :!class /Script/Engine.MaterialInstanceDynamic)
                    (Rendering|Material|SetVectorParameterValue _flipbookmid_6 "SheetCenter" (Math|Conversions|ToLinearColor(Vector) (Variables|Default|GetCurrentRenderLoc)) :!class /Script/Engine.MaterialInstanceDynamic)
                    (Rendering|Material|SetScalarParameterValue (Variables|RenderFlipbook|GetFlipbookMID) "Size" (Variables|Default|GetSectorSize) :!class /Script/Engine.MaterialInstanceDynamic)
                    (Rendering|Material|SetScalarParameterValue (Variables|RenderFlipbook|GetFlipbookMID) "FramesX" (Math|Conversions|ToFloat(Integer) (Variables|Render3DImposterSprites|GetFrames%20around%20Z%20rotation)) :!class /Script/Engine.MaterialInstanceDynamic)
                    (Rendering|Material|SetScalarParameterValue (Variables|RenderFlipbook|GetFlipbookMID) "FramesY" (Math|Conversions|ToFloat(Integer) (/ (Variables|Render3DImposterSprites|GetFrames%20around%20Z%20rotation) (Variables|Render3DImposterSprites|GetAspect%20Ratio%201%20by))) :!class /Script/Engine.MaterialInstanceDynamic)
                    (Rendering|Material|SetScalarParameterValue _flipbookmid_5 "Animation Phase" _0_1_phase :!class /Script/Engine.MaterialInstanceDynamic)
                    (Rendering|Material|SetScalarParameterValue _flipbookmid_5 "MeshScale" (Variables|Render3DImposterSprites|GetMeshScale) :!class /Script/Engine.MaterialInstanceDynamic)
                    (Rendering|Material|SetScalarParameterValue _flipbookmid_5 "UVIndex" (Variables|Render3DImposterSprites|GetUV%20Index) :!class /Script/Engine.MaterialInstanceDynamic)
                    (else
                      (Utilities|FlowControl|Sequence
                        (:then_0
                          (for _index (range (Rendering|Material|GetNumMaterials _staticmeshcomponent))
                            (bind _returnvalue_7 (Rendering|Material|CreateDynamicMaterialInstance (Variables|Default|GetStaticMeshComponent) _index (select (> _index (Utilities|Array|LastIndex (Variables|Render3DImposterSprites|GetImposter%20MaterialInstanceArray))) (Rendering|Material|GetMaterial _staticmeshcomponent _index) (Utilities|Array|Get(acopy) (Variables|Render3DImposterSprites|GetImposter%20MaterialInstanceArray) _index)) "None" :!class /Script/Engine.PrimitiveComponent))
                            (Variables|RenderFlipbook|SetFlipbookMID _returnvalue_7)))
                        (:then_1
                          _)))))
                (:NewEnumerator6
                  ;[0.22 0.22 0.22] Flipbook
                  ;;[1 1 1] Offsetting and Rotation
                  (Transformation|AddLocalRotation (Variables|Default|GetStaticMeshComponent) (Variables|RenderFlipbook|GetFlipbook%20Start%20Rotation) false false)
                  (Transformation|AddLocalRotation (Variables|Default|GetStaticMeshComponent) (Math|Vector|RotatorfromAxisandAngle (Variables|RenderFlipbook|GetFlipbook%20Rotation%20Axis%201) (* (Variables|RenderFlipbook|GetAxis%201%20Rotations) _returnvalue)) false false)
                  (Transformation|AddLocalRotation (Variables|Default|GetStaticMeshComponent) (Math|Vector|RotatorfromAxisandAngle (Variables|RenderFlipbook|GetFlipbook%20Rotation%20Axis%202) (* (Variables|RenderFlipbook|GetAxis%202%20Rotations) _returnvalue)) false false)
                  (Transformation|AddLocalOffset (Variables|Default|GetStaticMeshComponent) (* (Variables|Default|GetInitialOffset) (Variables|Default|GetInternalscalefactor)) false false)
                  (Transformation|SetRelativeScale3D (Variables|Default|GetStaticMeshComponent) (Math|Conversions|ToVector(Float) (Variables|Default|GetInternalscalefactor)))
                  ;;[1 1 1] Motion Vector Option
                  (if (Variables|RenderFlipbook|GetRender%20Motion%20Vectors)
                    ;;[1 1 1] Motion Vector passes
                    (Variables|RenderFlipbook|SetMaterialChoice (select (Variables|Default|GetUV%20DilationPass) (Variables|Default|GetMotionVectorMaterial) (Variables|Default|GetMotionVectorsApplied)))
                    (if (Variables|Default|GetUV%20DilationPass)
                      (Utilities|FlowControl|Sequence
                        (:then_0
                          ;;[0.1732 0.5569 1] MID Setup
                          (bind _returnvalue_8 (Rendering|Material|CreateDynamicMaterialInstance (Variables|RenderFlipbook|GetMaterialChoice) "None" "None" :!class /Script/Engine.KismetMaterialLibrary))
                          (Variables|RenderFlipbook|SetFlipbookMID _returnvalue_8))
                        (:then_1
                          ;;[0.4937 0.735 0] Motion Vector UV Layout
                          (Utilities|ApplyMIDToAllMeshMatIDs (Variables|Default|GetStaticMeshComponent) (Variables|RenderFlipbook|GetFlipbookMID))))
                      (else
                        ;;[1 1 1] Motion Vector Masked Setup
                        (bind _mid_2 (Utilities|CreateAndAssignMIDsFromMaskArrays (Variables|RenderFlipbook|GetMaterialChoice) (Variables|Default|GetStaticMeshComponent) (Variables|RenderFlipbook|GetOpacity%20MaskTextures) (Variables|RenderFlipbook|GetOpacity%20MaskChannels))
                          (:Loop%20Body%20Continue
                            (Variables|RenderFlipbook|SetFlipbookMID _mid_2)))))
                    (else
                      ;;[1 1 1] Flipbook Main Pass
                      (for _index_2 (range (Rendering|Material|GetNumMaterials _staticmeshcomponent_4))
                        (bind _returnvalue_9 (Rendering|Material|CreateDynamicMaterialInstance (Variables|Default|GetStaticMeshComponent) _index_2 (select (> _index_2 (Utilities|Array|LastIndex (Variables|RenderFlipbook|GetFlipbook%20MaterialInstance%20List))) (Rendering|Material|GetMaterial _staticmeshcomponent_4 _index_2) (Utilities|Array|Get(acopy) (Variables|RenderFlipbook|GetFlipbook%20MaterialInstance%20List) _index_2)) "None" :!class /Script/Engine.PrimitiveComponent))
                        (Variables|RenderFlipbook|SetFlipbookMID _returnvalue_9)
                        (Rendering|Material|SetVectorParameterValue _flipbookmid_4 "FrameCenter" (Math|Conversions|ToLinearColor(Vector) (Variables|Default|GetPolyCenter)) :!class /Script/Engine.MaterialInstanceDynamic)
                        (Rendering|Material|SetVectorParameterValue _flipbookmid_4 "SheetCenter" (Math|Conversions|ToLinearColor(Vector) (Variables|Default|GetCurrentRenderLoc)) :!class /Script/Engine.MaterialInstanceDynamic)
                        (Rendering|Material|SetScalarParameterValue _flipbookmid_3 "Size" (Variables|Default|GetSectorSize) :!class /Script/Engine.MaterialInstanceDynamic)
                        (Rendering|Material|SetScalarParameterValue _flipbookmid_3 "Animation Phase" _0_1_phase :!class /Script/Engine.MaterialInstanceDynamic)
                        (Rendering|Material|SetScalarParameterValue _flipbookmid "FramesX" (Math|Conversions|ToFloat(Integer) (Variables|RenderFlipbook|GetFlipbook%20Columns%20%28X%29)) :!class /Script/Engine.MaterialInstanceDynamic)
                        (Rendering|Material|SetScalarParameterValue _flipbookmid_2 "FramesY" (Math|Conversions|ToFloat(Integer) (Variables|RenderFlipbook|GetFlipbook%20Rows%20%28Y%29)) :!class /Script/Engine.MaterialInstanceDynamic)
                        (Rendering|Material|SetScalarParameterValue _flipbookmid_2 "MeshScale" (Variables|RenderFlipbook|GetFlipbook%20Mesh%20scale) :!class /Script/Engine.MaterialInstanceDynamic)
                        (Rendering|Material|SetScalarParameterValue _flipbookmid "UVIndex" (Variables|RenderFlipbook|GetUVIndex) :!class /Script/Engine.MaterialInstanceDynamic)))))))))))
    (:then_1
      (Utilities|FlowControl|Sequence
        (:then_0
          (if (Utilities|Flipbook-ImposterMotionVectorChooser)
            (Utilities|FlowControl|Sequence
              (:then_0
                ;[1 1 1] Dilation
                (for _index_3 (range _returnvalue_2)
                  (for _array_element (Variables|RenderFlipbook|GetUnwrappedMotionVectorMeshes)
                    (for _index_4 (range 8)
                      (CallFunction|CopyAMesh _array_element (+ (+ (/ (* (Utilities|Array|Get(acopy) (Variables|Default|GetOffsetVectors) _index_4) (* (/ (Math|Conversions|ToFloat(Integer) (+ _index_3 1)) (Math|Conversions|ToFloat(Integer) _returnvalue_2)) (Flipbook/ImposterCombiners|UVDilationAmountParamChooser))) (Variables|Debug|GetSheetSize)) (* "0.000000, 0,0.010000" (* (Math|Conversions|ToFloat(Integer) (+ _index_3 1)) 1))) (Math|Vector|SelectVector "0, 0,0.003000" "0, 0, 0" (> _index_4 3))) (Utilities|Array|MakeArray 0))))))
              (:then_1
                ;[1 1 1] Intensity Boost
                (Rendering|Material|SetScalarParameterValue "/Engine/ArtTools/RenderToTexture/Materials/MPC_R2T.MPC_R2T" "RenderMotionVectorMultiply" (Flipbook/ImposterCombiners|Flipbook-ImposterParamChooser (Variables|Render3DImposterSprites|GetMotion%20Vector%20IntensityBoost) (Variables|RenderFlipbook|GetMotionVectorIntensityBoost)) :!class /Script/Engine.KismetMaterialLibrary)))))
        (:then_1
          ;[1 1 1] Set Render Targets
          (Variables|/Script/Engine.SceneCaptureComponent2D|SetTextureTarget "/Engine/ArtTools/RenderToTexture/Textures/TilingTexturePreview.TilingTexturePreview" (Variables|Default|GetTilingPreviewCapture))
          (Variables|/Script/Engine.SceneCaptureComponent2D|SetTextureTarget "/Engine/ArtTools/RenderToTexture/Textures/FlipbookMotionVectors/Position1.Position1" (Variables|Default|GetSceneCapture%20Motion%20Position%201))
          ;[1 1 1] New Debug using Scene Capture
          (bind _returnvalue_10 (Rendering|Material|CreateDynamicMaterialInstance (Variables|Default|GetPreviewMesh) 0 (Utilities|Select :Index (Variables|CaptureSettings|GetRender%20Type) :NewEnumerator0 (Utilities|Select :Index (Variables|Render3DImposterSprites|GetImposterEnum) :NewEnumerator0 (Variables|Default|GetImposter-3D%20Preview%20Mat) :NewEnumerator1 (Variables|Default|GetImposter-SingleAxis%20Preview%20Mat)) :NewEnumerator6 (Variables|Default|GetFlipbook%20Preview%20Mat)) "None" :!class /Script/Engine.PrimitiveComponent))
          (Rendering|Material|SetScalarParameterValue _returnvalue_10 "Columns" (Math|Conversions|ToFloat(Integer) _x_2) :!class /Script/Engine.MaterialInstanceDynamic)
          (Rendering|Material|SetScalarParameterValue _returnvalue_10 "Rows" (Math|Conversions|ToFloat(Integer) _y_2) :!class /Script/Engine.MaterialInstanceDynamic)
          (Rendering|Material|SetScalarParameterValue _returnvalue_10 "Speed" (Variables|RenderFlipbook|GetPreview%20Speed) :!class /Script/Engine.MaterialInstanceDynamic)
          (Rendering|SetVisibility (Variables|Default|GetPreviewMesh) true false)
          (Transformation|SetWorldScale3D (Variables|Default|GetPlaneMesh) (Math|Conversions|ToVector(Float) (* (Variables|Debug|GetSheetSize) 1.05))))
        (:then_2
          (return))))))

(fn ImposterSprites
  (bind _imposter_materialinstancearray (Variables|Render3DImposterSprites|GetImposter%20MaterialInstanceArray))
  (bind _returnvalue (Utilities|Array|Length _imposter_materialinstancearray))
  (bind _returnvalue_2 (Transformation|GetActorTransform))
  (bind _aspect_ratio_1_by (Variables|Render3DImposterSprites|GetAspect%20Ratio%201%20by))
  (bind _frames_around_z_rotation (Variables|Render3DImposterSprites|GetFrames%20around%20Z%20rotation))
  (bind _returnvalue_3 (/ _frames_around_z_rotation _aspect_ratio_1_by))
  (bind _sectorsize (Variables|Default|GetSectorSize))
  (bind _stadium_offsetting (Variables|Render3DImposterSprites|GetStadium%20Offsetting))
  (bind _staticmeshcomponent (Variables|Default|GetStaticMeshComponent))
  (bind _frames_around_z_rotation_2 (Variables|Render3DImposterSprites|GetFrames%20around%20Z%20rotation))
  (bind _textsize (Variables|RenderDepthMap|GetTextSize))
  (bind _previewmesh (Variables|Default|GetPreviewMesh))
  (Utilities|FlowControl|Sequence
    (:then_0
      (Variables|Default|SetSectorSize (/ (Variables|Debug|GetSheetSize) (Math|Conversions|ToFloat(Integer) (Variables|Render3DImposterSprites|GetFrames%20around%20Z%20rotation))))
      (if (not (Variables|Render3DImposterSprites|GetUse%20Level%20Placed%20Meshes))
        (bind (_current_point _total_points _0_1_phase _poly_center _x_row _y_row) (Utilities|2DGridExecutionMacro :Sector%20Size (Variables|Default|GetSectorSize) :X _frames_around_z_rotation :Y _returnvalue_3)
          (:Per%20Poly%20Execution
            (bind _returnvalue_4 (AddComponent|NODE_AddStaticMeshComponent-1 :bManualAttachment true :RelativeTransform (Math|Conversions|ToTransform(Vector) (Math|Transform|TransformLocation _returnvalue_2 (Math|Vector|SelectVector (+ _poly_center (* "0, 0,1.000000" (* (Math|Conversions|ToFloat(Integer) _y_row) _sectorsize))) _poly_center _stadium_offsetting)))))
            (Variables|Default|SetStaticMeshComponent _returnvalue_4)
            (Components|StaticMesh|SetStaticMesh (Variables|Default|GetStaticMeshComponent) (Variables|Render3DImposterSprites|GetImposter%20Static%20Mesh))
            (Utilities|FlowControl|Sequence
              (:then_0
                (Variables|Default|SetInitialOffset (- (Math|Transform|TransformLocation _returnvalue_2 (Math|Vector|SelectVector (+ _poly_center (* "0, 0,1.000000" (* (Math|Conversions|ToFloat(Integer) _y_row) _sectorsize))) _poly_center _stadium_offsetting)) (Collision|GetComponentBounds _returnvalue_4)))
                (Variables|Default|Setinitialscale (.SphereRadius (Collision|GetComponentBounds _returnvalue_4)))
                (Variables|Default|SetInternalscalefactor (* (/ (Variables|Default|GetSectorSize) (* (.SphereRadius (Collision|GetComponentBounds _returnvalue_4)) 2)) (Variables|Render3DImposterSprites|GetMeshScale)))
                (Transformation|AddLocalRotation (Variables|Default|GetStaticMeshComponent) (Transformation|GetActorRotation) false false)
                (switch Utilities|FlowControl|Switch|SwitchonRenderToTexture_Imposter_Enum (Variables|Render3DImposterSprites|GetImposterEnum)
                  (:NewEnumerator0
                    ;[1 1 1] 3D Rotation
                    (Transformation|AddLocalRotation (Variables|Default|GetStaticMeshComponent) (Math|Rotator|MakeRotator (* (/ (Math|Conversions|ToFloat(Integer) _y_row) (Math|Conversions|ToFloat(Integer) _returnvalue_3)) -180)) false false)
                    (Transformation|AddLocalRotation (Variables|Default|GetStaticMeshComponent) (Math|Rotator|MakeRotator 0.0 0.0 (* (/ (Math|Conversions|ToFloat(Integer) _x_row) (Math|Conversions|ToFloat(Integer) _frames_around_z_rotation)) 360)) false false))
                  (:NewEnumerator1
                    ;[1 1 1] Single Rotation Axis
                    (Transformation|AddLocalRotation _staticmeshcomponent (Math|Rotator|MakeRotfromZ (Variables|Render3DImposterSprites|GetSingle%20Rotation%20Axis)) false false)
                    (Transformation|AddLocalRotation _staticmeshcomponent (Math|Rotator|MakeRotator 0.0 0.0 (* _0_1_phase 360)) false false))))
              (:then_1
                ;[1 1 1] Offsetting
                (Transformation|AddLocalOffset (Variables|Default|GetStaticMeshComponent) (* (Variables|Default|GetInitialOffset) (Variables|Default|GetInternalscalefactor)) false false)
                (Transformation|SetRelativeScale3D (Variables|Default|GetStaticMeshComponent) (Math|Conversions|ToVector(Float) (Variables|Default|GetInternalscalefactor)))
                (Utilities|FlowControl|Sequence
                  (:then_0
                    ;[1 1 1] Assign Material Instances
                    (if (> _returnvalue 1)
                      (for _index (range 1 (+ _returnvalue 1))
                        (Rendering|Material|SetMaterial (Variables|Default|GetStaticMeshComponent) (+ _index 0) (Utilities|Array|Get(acopy) (Variables|Render3DImposterSprites|GetImposter%20MaterialInstanceArray) (+ _index 0))))
                      (else
                        (Rendering|Material|SetMaterial (Variables|Default|GetStaticMeshComponent) 0 (Utilities|Array|Get(acopy) (Variables|Render3DImposterSprites|GetImposter%20MaterialInstanceArray))))))
                  (:then_1
                    ;[1 1 1] Store array of imposter info
                    (Utilities|Array|Add (Variables|Render3DImposterSprites|GetImposterArray) (Math|Transform|MakeTransform (Transformation|GetWorldLocation (Variables|Default|GetStaticMeshComponent)) (Transformation|GetWorldRotation (Variables|Default|GetStaticMeshComponent)) (Transformation|GetWorldScale (Variables|Default|GetStaticMeshComponent))))
                    ;[1 1 1] Imposter Depth Mask Setup
                    (if (Variables|Render3DImposterSprites|GetDebug%20Depth)
                      (for _index_2 (range (+ (Rendering|Material|GetNumMaterials (Variables|Default|GetStaticMeshComponent)) 1))
                        (bind _returnvalue_5 (Rendering|Material|CreateDynamicMaterialInstance "/Engine/ArtTools/RenderToTexture/Materials/M_RenderNormalizedDepth_01_Inst.M_RenderNormalizedDepth_01_Inst" "None" "None" :!class /Script/Engine.KismetMaterialLibrary))
                        (Rendering|Material|SetTextureParameterValue _returnvalue_5 "Opacity Mask" (select (> _index_2 (Utilities|Array|LastIndex (Variables|Render3DImposterSprites|GetOpacity%20Mask%20Textures))) "/Engine/ArtTools/RenderToTexture/Textures/T_EV_BlankWhite_01.T_EV_BlankWhite_01" (Utilities|Array|Get(acopy) (Variables|Render3DImposterSprites|GetOpacity%20Mask%20Textures) _index_2)))
                        (Rendering|Material|SetVectorParameterValue _returnvalue_5 "Mask Channel" (Math|Color|SelectColor (Math|Color|MakeColor 0.0) (Utilities|Array|Get(acopy) (Variables|Render3DImposterSprites|GetOpacity%20Mask%20Channels) _index_2) (> _index_2 (Utilities|Array|LastIndex (Variables|Render3DImposterSprites|GetOpacity%20Mask%20Channels)))) :!class /Script/Engine.MaterialInstanceDynamic)
                        (Rendering|Material|SetMaterial (Variables|Default|GetStaticMeshComponent) _index_2 _returnvalue_5))))
                  (:then_2
                    ;[0.2707 0.1032 0.34] Size and offset info to use in final material
                    (bind _returnvalue_6 (AddComponent|NODE_AddTextRenderComponent-2 :bManualAttachment true :RelativeTransform (Math|Transform|MakeTransform (* (Math|Vector|MakeVector (Variables|Debug|GetSheetSize)) 0.625) "90.000000,180.000000,90.000000" (Math|Conversions|ToVector(Float) _textsize))))
                    (Rendering|Components|TextRender|SetText _returnvalue_6 (Utilities|Text|ToText(String) (Utilities|String|BuildString(Float) "Boundsphere Radius: " "" (Variables|Default|Getinitialscale) " ")))
                    (bind _returnvalue_7 (AddComponent|NODE_AddTextRenderComponent-3 :bManualAttachment true :RelativeTransform (Math|Transform|MakeTransform (+ (* (Math|Vector|MakeVector (Variables|Debug|GetSheetSize)) 0.625) "0,-25.000000, 0") "90.000000,180.000000,90.000000" (Math|Conversions|ToVector(Float) _textsize))))
                    (Rendering|Components|TextRender|SetText _returnvalue_7 (Utilities|Text|ToText(String) (Utilities|String|Append "Offset Vector: " (Utilities|String|ToString(Vector) (Variables|Default|GetInitialOffset)))))))))))))
    (:then_1
      (bind _returnvalue_8 (Rendering|Material|CreateDynamicMaterialInstance "/Engine/ArtTools/RenderToTexture/Materials/Imposters/M_ImposterTest_01_Inst.M_ImposterTest_01_Inst" "None" "None" :!class /Script/Engine.KismetMaterialLibrary))
      (Rendering|Material|SetScalarParameterValue _returnvalue_8 "Columns" (Math|Conversions|ToFloat(Integer) _frames_around_z_rotation_2) :!class /Script/Engine.MaterialInstanceDynamic)
      (Rendering|Material|SetScalarParameterValue _returnvalue_8 "Rows" (Math|Conversions|ToFloat(Integer) (/ _frames_around_z_rotation_2 (Variables|Render3DImposterSprites|GetAspect%20Ratio%201%20by))) :!class /Script/Engine.MaterialInstanceDynamic)
      (Rendering|Material|SetMaterial _previewmesh 0 _returnvalue_8)
      (Rendering|SetVisibility _previewmesh true false)
      (Variables|/Script/Engine.SceneCaptureComponent2D|SetTextureTarget "/Engine/ArtTools/RenderToTexture/Textures/TilingTexturePreview.TilingTexturePreview" (Variables|Default|GetTilingPreviewCapture))
      (return))))

(fn NewMesh
  (params (Mesh :type Static%20Mesh%20Object%20Reference) (WorldPos :type Vector) (Material :type Material%20Interface%20Object%20Reference) (Scale3d :type Vector))
  (bind _returnvalue (AddComponent|NODE_AddStaticMeshComponent-7 :bManualAttachment true :RelativeTransform (Math|Transform|MakeTransform "5000.000000, 0, 0" "0, 0, 0" Scale3d)))
  (Components|StaticMesh|SetStaticMesh _returnvalue Mesh)
  (Transformation|SetWorldLocation _returnvalue WorldPos false false)
  (bind _returnvalue_2 (Rendering|Material|CreateDynamicMaterialInstance Material "None" "None" :!class /Script/Engine.KismetMaterialLibrary))
  (Rendering|Material|SetScalarParameterValue _returnvalue_2 "Size" (Variables|Default|GetSectorSize) :!class /Script/Engine.MaterialInstanceDynamic)
  (Rendering|Material|SetVectorParameterValue _returnvalue_2 "SheetCenter" (Math|Conversions|ToLinearColor(Vector) (Variables|Default|GetCurrentRenderLoc)) :!class /Script/Engine.MaterialInstanceDynamic)
  (Rendering|Material|SetMaterial _returnvalue 0 _returnvalue_2)
  (return _returnvalue))

(fn ParticleSystemFrameDump
  (Effects|SpawnEmitterAttached (Variables|ParticleSystemFrameDump|GetParticleSystem) (Variables|Default|GetRoot)))

(fn RenderDepthMap
  (bind _directional_light (Variables|RenderDepthMap|GetDirectional%20Light))
  (bind _returnvalue (Utilities|Equal(Object) _directional_light 0))
  (bind _lightvector (Variables|RenderDepthMap|GetLightVector))
  (bind _returnvalue_2 (Math|Vector|Equal(Vector) _lightvector "1.000000, 0, 0" 0.0001))
  (bind _returnvalue_3 (Math|Vector|SelectVector "1.000000, 0,0.000200" _lightvector _returnvalue_2))
  (bind _returnvalue_4 (Math|Vector|Normalize _returnvalue_3))
  (bind _returnvalue_5 (Transformation|GetActorRotation _directional_light))
  (bind _returnvalue_6 (Math|Vector|GetForwardVector _returnvalue_5))
  (bind (_x _y _z) (Math|Rotator|BreakRotIntoAxes (.rotation (Transformation|GetActorTransform))))
  (bind _staticmeshcomponent (Variables|Default|GetStaticMeshComponent))
  (bind (_origin _boxextent _sphereradius) (Collision|GetComponentBounds _staticmeshcomponent))
  (bind _returnvalue_7 (* _sphereradius 2))
  (bind _sheetsize (Variables|Debug|GetSheetSize))
  (bind _returnvalue_8 (* _sheetsize 0.5))
  (bind _staticmeshcomponent_2 (Variables|Default|GetStaticMeshComponent))
  (bind _depthmid (Variables|RenderDepthMap|GetDepthMID))
  (bind _depthmid_2 (Variables|RenderDepthMap|GetDepthMID))
  (bind _textsize (Variables|RenderDepthMap|GetTextSize))
  (bind _fit_vector_and_scale_info_onto_texture (Variables|RenderDepthMap|GetFit%20Vector%20and%20scale%20info%20onto%20texture))
  (bind _text_locations (Variables|RenderDepthMap|GetText%20Locations))
  (Rendering|SetVisibility (Variables|Default|GetPlaneMesh) true false)
  (if (Variables|RenderDepthMap|GetUse%20Level%20Meshes%20for%20Depth)
    (bind _returnvalue_9 (AddComponent|NODE_AddStaticMeshComponent-3)) ; Debug Sphere for bounds
    (Transformation|SetWorldScale3D _returnvalue_9 (Math|Conversions|ToVector(Float) (/ (Variables|Debug|GetSheetSize) 2)))
    (else
      (bind _returnvalue_10 (AddComponent|NODE_AddStaticMeshComponent-2 :bManualAttachment true))
      (Variables|Default|SetStaticMeshComponent _returnvalue_10)
      (Components|StaticMesh|SetStaticMesh (Variables|Default|GetStaticMeshComponent) (Variables|RenderDepthMap|GetDepth%20Map%20Static%20Mesh))
      (Variables|Default|SetInitialOffset (- "0, 0, 0" _origin))
      (Transformation|SetWorldRotation (Variables|Default|GetStaticMeshComponent) (Math|Rotator|MakeRotationfromAxes (Math|Vector|Normalize (Math|Vector|CrossProduct (Math|Vector|Normalize (Math|Vector|CrossProduct (Math|Vector|GetForwardVector (Math|Rotator|MakeRotationfromAxes (Math|Vector|SelectVector (* _returnvalue_4 "1.000000,1.000000,-1.000000") (* _returnvalue_6 "1.000000,1.000000,-1.000000") _returnvalue) (Math|Vector|Normalize (Math|Vector|CrossProduct (Math|Vector|SelectVector (* _returnvalue_4 "1.000000,1.000000,-1.000000") (* _returnvalue_6 "1.000000,1.000000,-1.000000") _returnvalue) (* "-1.000000,0.000000, 0" "1.000000,1.000000,1.000000"))) (Math|Vector|Normalize (Math|Vector|CrossProduct (Math|Vector|Normalize (Math|Vector|CrossProduct (Math|Vector|SelectVector (* _returnvalue_4 "1.000000,1.000000,-1.000000") (* _returnvalue_6 "1.000000,1.000000,-1.000000") _returnvalue) (* "-1.000000,0.000000, 0" "1.000000,1.000000,1.000000"))) (* "-1.000000,0.000000, 0" "1.000000,1.000000,1.000000"))))) "1.000000, 0,0.000000")) (Math|Vector|GetForwardVector (Math|Rotator|MakeRotationfromAxes (Math|Vector|SelectVector (* _returnvalue_4 "1.000000,1.000000,-1.000000") (* _returnvalue_6 "1.000000,1.000000,-1.000000") _returnvalue) (Math|Vector|Normalize (Math|Vector|CrossProduct (Math|Vector|SelectVector (* _returnvalue_4 "1.000000,1.000000,-1.000000") (* _returnvalue_6 "1.000000,1.000000,-1.000000") _returnvalue) (* "-1.000000,0.000000, 0" "1.000000,1.000000,1.000000"))) (Math|Vector|Normalize (Math|Vector|CrossProduct (Math|Vector|Normalize (Math|Vector|CrossProduct (Math|Vector|SelectVector (* _returnvalue_4 "1.000000,1.000000,-1.000000") (* _returnvalue_6 "1.000000,1.000000,-1.000000") _returnvalue) (* "-1.000000,0.000000, 0" "1.000000,1.000000,1.000000"))) (* "-1.000000,0.000000, 0" "1.000000,1.000000,1.000000"))))))) (Math|Vector|Normalize (Math|Vector|CrossProduct (Math|Vector|GetForwardVector (Math|Rotator|MakeRotationfromAxes (Math|Vector|SelectVector (* _returnvalue_4 "1.000000,1.000000,-1.000000") (* _returnvalue_6 "1.000000,1.000000,-1.000000") _returnvalue) (Math|Vector|Normalize (Math|Vector|CrossProduct (Math|Vector|SelectVector (* _returnvalue_4 "1.000000,1.000000,-1.000000") (* _returnvalue_6 "1.000000,1.000000,-1.000000") _returnvalue) (* "-1.000000,0.000000, 0" "1.000000,1.000000,1.000000"))) (Math|Vector|Normalize (Math|Vector|CrossProduct (Math|Vector|Normalize (Math|Vector|CrossProduct (Math|Vector|SelectVector (* _returnvalue_4 "1.000000,1.000000,-1.000000") (* _returnvalue_6 "1.000000,1.000000,-1.000000") _returnvalue) (* "-1.000000,0.000000, 0" "1.000000,1.000000,1.000000"))) (* "-1.000000,0.000000, 0" "1.000000,1.000000,1.000000"))))) "1.000000, 0,0.000000")) (Math|Vector|GetForwardVector (Math|Rotator|MakeRotationfromAxes (Math|Vector|SelectVector (* _returnvalue_4 "1.000000,1.000000,-1.000000") (* _returnvalue_6 "1.000000,1.000000,-1.000000") _returnvalue) (Math|Vector|Normalize (Math|Vector|CrossProduct (Math|Vector|SelectVector (* _returnvalue_4 "1.000000,1.000000,-1.000000") (* _returnvalue_6 "1.000000,1.000000,-1.000000") _returnvalue) (* "-1.000000,0.000000, 0" "1.000000,1.000000,1.000000"))) (Math|Vector|Normalize (Math|Vector|CrossProduct (Math|Vector|Normalize (Math|Vector|CrossProduct (Math|Vector|SelectVector (* _returnvalue_4 "1.000000,1.000000,-1.000000") (* _returnvalue_6 "1.000000,1.000000,-1.000000") _returnvalue) (* "-1.000000,0.000000, 0" "1.000000,1.000000,1.000000"))) (* "-1.000000,0.000000, 0" "1.000000,1.000000,1.000000")))))) false false)
      (Variables|Default|SetInternalscalefactor (/ (Variables|Debug|GetSheetSize) (* _returnvalue_7 (Variables|RenderDepthMap|GetScale%20XY))))
      (Variables|RenderDepthMap|SetDepthMeshInitialSize (Math|Vector|MakeVector _returnvalue_7 (* (Math|Float|Max(Float) (.x _boxextent) (.y _boxextent)) 2)))
      (Transformation|SetWorldScale3D (Variables|Default|GetStaticMeshComponent) (Math|Conversions|ToVector(Float) (Variables|Default|GetInternalscalefactor)))
      (Transformation|AddLocalOffset _staticmeshcomponent_2 (* (Variables|Default|GetInitialOffset) (Variables|Default|GetInternalscalefactor)) false false)
      (Transformation|AddWorldOffset _staticmeshcomponent_2 (* _z (* (Variables|Debug|GetSheetSize) 0.5)) false false)
      (Transformation|AddRelativeLocation (Variables|Default|GetStaticMeshComponent) (Transformation|GetActorLocation) false false)
      ;[1 1 1] Material Setup
      (for _index (range (Rendering|Material|GetNumMaterials (Variables|Default|GetStaticMeshComponent)))
        (bind _returnvalue_11 (Rendering|Material|CreateDynamicMaterialInstance (Variables|Default|GetStaticMeshComponent) (+ _index 0) "/Engine/ArtTools/RenderToTexture/Materials/M_RenderNormalizedDepth_01.M_RenderNormalizedDepth_01" "None" :!class /Script/Engine.PrimitiveComponent))
        (Variables|RenderDepthMap|SetDepthMID _returnvalue_11)
        (Rendering|Material|SetScalarParameterValue _depthmid_2 "Scale Z" (Variables|RenderDepthMap|GetScale%20Z) :!class /Script/Engine.MaterialInstanceDynamic)
        (Rendering|Material|SetScalarParameterValue _depthmid_2 "Size" (Variables|Debug|GetSheetSize) :!class /Script/Engine.MaterialInstanceDynamic)
        (Rendering|Material|SetVectorParameterValue _depthmid_2 "Blueprint Offset" (Math|Conversions|ToLinearColor(Vector) (Math|Conversions|ToVector(LinearColor) (Math|Conversions|ToLinearColor(Vector) (* (Variables|Default|GetInitialOffset) (Variables|Default|GetInternalscalefactor))))) :!class /Script/Engine.MaterialInstanceDynamic)
        (Rendering|Material|SetTextureParameterValue _depthmid "Opacity Mask" (Utilities|Array|Get(acopy) (Variables|RenderDepthMap|GetDepth%20Material%20Mask%20Textures) (+ _index 0)))
        (Rendering|Material|SetVectorParameterValue _depthmid "Mask Channel" (Utilities|Array|Get(acopy) (Variables|RenderDepthMap|GetMask%20Channel) (Math|Integer|Min(Integer) (+ _index 0) (Utilities|Array|LastIndex (Variables|RenderDepthMap|GetMask%20Channel)))) :!class /Script/Engine.MaterialInstanceDynamic))))
  (Variables|Debug|SetBufferCommands "r.BufferVisualizationOverviewTargets SceneColor")
  (Variables|Default|SetShotCommand (Utilities|String|Append "highresshot " (Utilities|String|ToString(Integer) (Math|Integer|Clamp(Integer) (Variables|CaptureSettings|GetResolutionMultiplier) 1 8))))
  ;[0.2707 0.1032 0.34] Size info text for capture
  (bind _returnvalue_12 (AddComponent|NODE_AddTextRenderComponent-4 :bManualAttachment true :RelativeTransform (Math|Transform|MakeTransform (Math|Vector|SelectVector (* (Utilities|Array|Get(acopy) _text_locations) _returnvalue_8) (* (Math|Vector|MakeVector _sheetsize) 0.625) _fit_vector_and_scale_info_onto_texture) "90.000000,180.000000,90.000000" (Math|Conversions|ToVector(Float) _textsize))))
  (Rendering|Components|TextRender|SetText _returnvalue_12 (Utilities|Text|ToText(String) (Utilities|String|BuildString(Float) (Utilities|String|BuildString(Float) "Scale XY: " "" (Variables|RenderDepthMap|GetScale%20XY) " ") "Scale Z:" (Variables|RenderDepthMap|GetScale%20Z))))
  (bind _returnvalue_13 (AddComponent|NODE_AddTextRenderComponent-5 :bManualAttachment true :RelativeTransform (Math|Transform|MakeTransform (Math|Vector|SelectVector (* (Utilities|Array|Get(acopy) _text_locations 1) _returnvalue_8) (+ (* (Math|Vector|MakeVector (Variables|Debug|GetSheetSize)) 0.625) "0,-25.000000, 0") _fit_vector_and_scale_info_onto_texture) "90.000000,180.000000,90.000000" (Math|Conversions|ToVector(Float) _textsize))))
  (Rendering|Components|TextRender|SetText _returnvalue_13 (Utilities|Text|ToText(String) (Utilities|String|Append "LightVector: " (Utilities|String|BuildString(Float) (Utilities|String|ToString(Float) (.x (Math|Vector|SelectVector _returnvalue_4 _returnvalue_6 _returnvalue))) "," (.y (Math|Vector|SelectVector _returnvalue_4 _returnvalue_6 _returnvalue)) (Utilities|String|BuildString(Float) "" "," (.z (Math|Vector|SelectVector _returnvalue_4 _returnvalue_6 _returnvalue)) ",")))))
  (return 0))

(fn RenderLightmap
  (bind _lightmap_mesh (Variables|RenderLightmaps|GetLightmap%20Mesh))
  (bind _staticmeshcomponent (Variables|/Script/Engine.StaticMeshActor|GetStaticMeshComponent _lightmap_mesh))
  (bind _unwrap2sidedmidlista (Variables|RenderLightmaps2-sided|GetUnwrap2sidedMIDlistA))
  (Utilities|Array|Clear _unwrap2sidedmidlista)
  (for _index (range (+ (Rendering|Material|GetNumMaterials _staticmeshcomponent) 1))
    (bind _returnvalue (Rendering|Material|CreateDynamicMaterialInstance _staticmeshcomponent _index (Variables|RenderLightmaps2-sided|GetLightmap%20Unwrap%20Material) "None" :!class /Script/Engine.PrimitiveComponent))
    (Utilities|Array|Add _unwrap2sidedmidlista _returnvalue))
  (for _array_element (Variables|RenderLightmaps2-sided|GetUnwrap2sidedMIDlistA)
    (Rendering|Material|SetVectorParameterValue _array_element "RenderLocation" (Math|Conversions|ToLinearColor(Vector) (Transformation|GetActorLocation)) :!class /Script/Engine.MaterialInstanceDynamic)
    (Rendering|Material|SetScalarParameterValue _array_element "Unwrap" (Math|Conversions|ToFloat(Boolean) (Variables|RenderLightmaps|GetUnwrap%20Lightmap)) :!class /Script/Engine.MaterialInstanceDynamic)
    (Rendering|Material|SetScalarParameterValue _array_element "Size" (* (Variables|Debug|GetSheetSize) 0.5) :!class /Script/Engine.MaterialInstanceDynamic))
  (return))

(fn RenderLIghtmap2-sided
  (bind _unwrap_lightmap_2sided (Variables|RenderLightmaps2-sided|GetUnwrap%20Lightmap%202sided))
  (bind _returnvalue (Math|Conversions|ToFloat(Boolean) _unwrap_lightmap_2sided))
  (bind _returnvalue_2 (Transformation|GetActorLocation))
  (bind _returnvalue_3 (Math|Conversions|ToLinearColor(Vector) _returnvalue_2))
  (bind _sheetsize (Variables|Debug|GetSheetSize))
  (bind _returnvalue_4 (* _sheetsize 0.5))
  (bind _lightmap_mesh_2_sided__side_1 (Variables|RenderLightmaps2-sided|GetLightmap%20Mesh%202-sided,%20Side%201))
  (bind _lightmap_mesh_2_sided__side_2 (Variables|RenderLightmaps2-sided|GetLightmap%20Mesh%202-sided,%20Side%202))
  (bind _staticmeshcomponent (Variables|/Script/Engine.StaticMeshActor|GetStaticMeshComponent _lightmap_mesh_2_sided__side_1))
  (bind _staticmeshcomponent_2 (Variables|/Script/Engine.StaticMeshActor|GetStaticMeshComponent _lightmap_mesh_2_sided__side_2))
  (bind _lightmap_unwrap_material (Variables|RenderLightmaps2-sided|GetLightmap%20Unwrap%20Material))
  (Utilities|Array|Clear (Variables|RenderLightmaps2-sided|GetUnwrap2sidedMIDlistA))
  (Utilities|Array|Clear (Variables|RenderLightmaps2-sided|GetUnwrap2sidedMIDListB))
  (for _index (range (+ (Rendering|Material|GetNumMaterials _staticmeshcomponent) 1))
    (bind _returnvalue_5 (Rendering|Material|CreateDynamicMaterialInstance _staticmeshcomponent _index _lightmap_unwrap_material "None" :!class /Script/Engine.PrimitiveComponent))
    (Utilities|Array|Add (Variables|RenderLightmaps2-sided|GetUnwrap2sidedMIDlistA) _returnvalue_5))
  (for _index_2 (range (+ (Rendering|Material|GetNumMaterials _staticmeshcomponent_2) 1))
    (bind _returnvalue_6 (Rendering|Material|CreateDynamicMaterialInstance _staticmeshcomponent_2 _index_2 _lightmap_unwrap_material "None" :!class /Script/Engine.PrimitiveComponent))
    (Utilities|Array|Add (Variables|RenderLightmaps2-sided|GetUnwrap2sidedMIDListB) _returnvalue_6))
  (for _array_element (Variables|RenderLightmaps2-sided|GetUnwrap2sidedMIDlistA)
    (Rendering|Material|SetVectorParameterValue _array_element "RenderLocation" _returnvalue_3 :!class /Script/Engine.MaterialInstanceDynamic)
    (Rendering|Material|SetScalarParameterValue _array_element "Unwrap" _returnvalue :!class /Script/Engine.MaterialInstanceDynamic)
    (Rendering|Material|SetScalarParameterValue _array_element "Size" _returnvalue_4 :!class /Script/Engine.MaterialInstanceDynamic))
  (for _array_element_2 (Variables|RenderLightmaps2-sided|GetUnwrap2sidedMIDListB)
    (Rendering|Material|SetVectorParameterValue _array_element_2 "RenderLocation" _returnvalue_3 :!class /Script/Engine.MaterialInstanceDynamic)
    (Rendering|Material|SetScalarParameterValue _array_element_2 "Unwrap" _returnvalue :!class /Script/Engine.MaterialInstanceDynamic)
    (Rendering|Material|SetScalarParameterValue _array_element_2 "Size" _returnvalue_4 :!class /Script/Engine.MaterialInstanceDynamic))
  (return))

(fn RenderRandomTilingPhysicsDrop
  (CallFunction|SetupCollision true)
  (Utilities|Array|Clear (Variables|PhysicsGround|GetStaticMobilityMeshes))
  (Utilities|Array|Clear (Variables|PhysicsGround|GetCopied%20Mesh%20Array))
  (Rendering|Material|SetMaterial (Variables|Default|GetPlaneMesh) 0 (Variables|PhysicsGround|GetPhysGround%20Mat))
  (Rendering|SetVisibility (Variables|Default|GetPreviewMesh) true false)
  (Components|Activation|SetActive (Variables|Default|GetTilingPreviewCapture) true false)
  (if (> (Utilities|Array|Length (Variables|PhysicsGround|GetSavedPhysMeshList)) 0)
    (Utilities|Array|Clear (Variables|PhysicsGround|GetSavedMeshComponents))
    (for _array_element (Variables|PhysicsGround|GetSavedPhysMeshList)
      (bind _returnvalue (AddComponent|NODE_AddStaticMeshComponent-4 :bManualAttachment true))
      (Transformation|SetWorldTransform _returnvalue (.Transform_9_8EA1546242BACCE5DB84C7B34C56F044 (Utilities|Struct|BreakPhysMesh _array_element)) false false)
      (Components|StaticMesh|SetStaticMesh _returnvalue (Utilities|Struct|BreakPhysMesh _array_element))
      (Utilities|Array|Add (Variables|PhysicsGround|GetSavedMeshComponents) _returnvalue)
      (if (Variables|PhysicsGround|GetDebug%20Displacement%20Depth)
        (bind _returnvalue_2 (Rendering|Material|CreateDynamicMaterialInstance "/Engine/ArtTools/RenderToTexture/Materials/M_RenderCustomHeightmap_01.M_RenderCustomHeightmap_01" "None" "None" :!class /Script/Engine.KismetMaterialLibrary))
        (Rendering|Material|SetScalarParameterValue _returnvalue_2 "Min" (Variables|PhysicsGround|GetDisplacement%20Min) :!class /Script/Engine.MaterialInstanceDynamic)
        (Rendering|Material|SetScalarParameterValue _returnvalue_2 "Max" (Variables|PhysicsGround|GetDisplacement%20Max) :!class /Script/Engine.MaterialInstanceDynamic)
        (Rendering|Material|SetTextureParameterValue _returnvalue_2 "Displacement" "")
        (Rendering|Material|SetScalarParameterValue _returnvalue_2 "Displacement Height" (Variables|RenderTilingMaterialfromMeshes|GetDisplacement%20Texture%20Height) :!class /Script/Engine.MaterialInstanceDynamic)
        (Rendering|Material|SetMaterial _returnvalue 0 _returnvalue_2))))
  (Utilities|FlowControl|Sequence
    (:then_0
      (if (Variables|PhysicsGround|GetKeep%20Tiling%20Hand%20Placed%20Meshes)
        (CallFunction|RenderTilingMaterialFromMeshes)))
    (:then_1
      ;[1 1 1] Render Target Preview setup
      (Variables|/Script/Engine.SceneCaptureComponent2D|SetTextureTarget "/Engine/ArtTools/RenderToTexture/Textures/TilingTexturePreview.TilingTexturePreview" (Variables|Default|GetTilingPreviewCapture))
      (bind _returnvalue_3 (Rendering|Material|CreateDynamicMaterialInstance (Variables|Default|GetPreviewMesh) 0 "/Engine/ArtTools/RenderToTexture/Materials/M_TilingPreviewMat_01.M_TilingPreviewMat_01" "None" :!class /Script/Engine.PrimitiveComponent))
      (Rendering|Material|SetScalarParameterValue _returnvalue_3 "Tiling" (Variables|PhysicsGround|GetTilingAmount) :!class /Script/Engine.MaterialInstanceDynamic)
      (Rendering|SetVisibility (Variables|Default|GetPreviewMesh) true false))))

(fn RenderTilingMaterial
  (Rendering|Material|CreateDynamicMaterialInstance (Variables|Default|GetPlaneMesh) 0 (Variables|RenderMaterial|GetMaterial%20Instance) "None" :!class /Script/Engine.PrimitiveComponent))

(fn RenderTilingMaterialfromMeshes
  (bind _current_component (Variables|Default|GetCurrent%20Component))
  (bind _copied_mesh_array (Variables|PhysicsGround|GetCopied%20Mesh%20Array))
  (bind _current_component_2 (Variables|Default|GetCurrent%20Component))
  (Utilities|Array|Clear (Variables|PhysicsGround|GetCopied%20Mesh%20Array))
  (for _array_element (Variables|RenderTilingMaterialfromMeshes|GetTilingMesh%20List)
    (bind _returnvalue (AddComponent|NODE_AddStaticMeshComponent-6 :RelativeTransform (.Transform_5_EC8FA7FF4AB65DB53669249F2EE91ED0 (Utilities|Struct|BreakTilingMesh _array_element))))
    (Variables|Default|SetCurrent%20Component _returnvalue)
    (Components|StaticMesh|SetStaticMesh (Variables|Default|GetCurrent%20Component) (Utilities|Struct|BreakTilingMesh _array_element))
    (if (Variables|RenderTilingMaterialfromMeshes|GetDebugDisplacementDepth)
      (bind _returnvalue_2 (Rendering|Material|CreateDynamicMaterialInstance "/Engine/ArtTools/RenderToTexture/Materials/M_RenderCustomHeightmap_01.M_RenderCustomHeightmap_01" "None" "None" :!class /Script/Engine.KismetMaterialLibrary))
      (Rendering|Material|SetScalarParameterValue _returnvalue_2 "Min" (Variables|RenderTilingMaterialfromMeshes|GetDisplacementMin) :!class /Script/Engine.MaterialInstanceDynamic)
      (Rendering|Material|SetScalarParameterValue _returnvalue_2 "Max" (Variables|RenderTilingMaterialfromMeshes|GetDisplacementMax) :!class /Script/Engine.MaterialInstanceDynamic)
      (Rendering|Material|SetTextureParameterValue _returnvalue_2 "Displacement" (.DisplacementTexture_13_FF80FB2A4568E664C650799B803BBBCB (Utilities|Struct|BreakTilingMesh _array_element)))
      (Rendering|Material|SetScalarParameterValue _returnvalue_2 "Displacement Height" (Variables|RenderTilingMaterialfromMeshes|GetDisplacement%20Texture%20Height) :!class /Script/Engine.MaterialInstanceDynamic)
      (Rendering|Material|SetMaterial _current_component 0 _returnvalue_2)
      (else
        (Rendering|Material|SetMaterial _current_component 0 (.Material_8_D7CA231C402F4C3E21216DBFF0C7B831 (Utilities|Struct|BreakTilingMesh _array_element)))))
    (Rendering|SetVisibility _current_component (.Visible_10_3155D32A40830AA72468EBB353E0197E (Utilities|Struct|BreakTilingMesh _array_element)) false)
    (Utilities|Array|Add _copied_mesh_array (Variables|Default|GetCurrent%20Component))
    (if (.Visible_10_3155D32A40830AA72468EBB353E0197E (Utilities|Struct|BreakTilingMesh _array_element))
      ;[0.0735 0.4957 1] Copy Bottom Right
      (CallFunction|CopyAMesh _current_component_2 "-1.000000, 0,0.000000" _copied_mesh_array)
      (CallFunction|CopyAMesh _current_component_2 "0,-1.000000,0.000000" _copied_mesh_array)
      (CallFunction|CopyAMesh _current_component_2 "-1.000000,-1.000000,0.000000" _copied_mesh_array)))
  (if (Variables|RenderTilingMaterialfromMeshes|GetPreview%20Tiling)
    ;[1 1 1] Render Target Preview setup
    (Variables|/Script/Engine.SceneCaptureComponent2D|SetTextureTarget "/Engine/ArtTools/RenderToTexture/Textures/TilingTexturePreview.TilingTexturePreview" (Variables|Default|GetTilingPreviewCapture))
    (bind _returnvalue_3 (Rendering|Material|CreateDynamicMaterialInstance (Variables|Default|GetPreviewMesh) 0 "/Engine/ArtTools/RenderToTexture/Materials/M_TilingPreviewMat_01.M_TilingPreviewMat_01" "None" :!class /Script/Engine.PrimitiveComponent))
    (Rendering|Material|SetScalarParameterValue _returnvalue_3 "Tiling" (Variables|RenderTilingMaterialfromMeshes|GetTiling%20Amount) :!class /Script/Engine.MaterialInstanceDynamic)
    (Rendering|SetVisibility (Variables|Default|GetPreviewMesh) true false)
    (else
      (Variables|/Script/Engine.SceneCaptureComponent2D|SetTextureTarget "" (Variables|Default|GetTilingPreviewCapture)))))

(fn RenderUnwrappedMeshToTextures
  (bind _render_to_texture_mesh (Variables|RenderUnwrappedMesh|GetRender%20to%20Texture%20Mesh))
  (bind _staticmeshcomponent (Variables|/Script/Engine.StaticMeshActor|GetStaticMeshComponent _render_to_texture_mesh))
  (for _index (range (+ (Rendering|Material|GetNumMaterials _staticmeshcomponent) 1))
    (bind _returnvalue (Rendering|Material|CreateDynamicMaterialInstance _staticmeshcomponent _index 0 "None" :!class /Script/Engine.PrimitiveComponent))
    (Rendering|Material|SetVectorParameterValue _returnvalue "RenderLocation" (Math|Conversions|ToLinearColor(Vector) (Transformation|GetActorLocation)) :!class /Script/Engine.MaterialInstanceDynamic)
    (Rendering|Material|SetScalarParameterValue _returnvalue "Unwrap" (Math|Conversions|ToFloat(Boolean) (Variables|RenderUnwrappedMesh|GetUnwrap)) :!class /Script/Engine.MaterialInstanceDynamic)
    (Rendering|Material|SetScalarParameterValue _returnvalue "Size" (* (Variables|Debug|GetSheetSize) 0.5) :!class /Script/Engine.MaterialInstanceDynamic))
  (return))

(fn SetBufferCommands
  (bind _specular (Variables|CaptureSettings|GetSpecular))
  (bind _returnvalue (Utilities|String|SelectString ", Specular" "" _specular))
  (bind _opacity (Variables|CaptureSettings|GetOpacity))
  (bind _returnvalue_2 (Utilities|String|SelectString ", Opacity" "" _opacity))
  (bind _roughness (Variables|CaptureSettings|GetRoughness))
  (bind _returnvalue_3 (Utilities|String|SelectString ", Roughness" "" _roughness))
  (bind _ambientocclusion (Variables|CaptureSettings|GetAmbientOcclusion))
  (bind _returnvalue_4 (Utilities|String|SelectString ", AmbientOcclusion" "" _ambientocclusion))
  (bind _materialambientocclusion (Variables|CaptureSettings|GetMaterialAmbientOcclusion))
  (bind _returnvalue_5 (Utilities|String|SelectString ", MaterialAmbientOcclusion" "" _materialambientocclusion))
  (bind _basecolor (Variables|CaptureSettings|GetBaseColor))
  (bind _returnvalue_6 (Utilities|String|SelectString "BaseColor" "" _basecolor))
  (bind _returnvalue_7 (Utilities|String|Append _returnvalue_6 _returnvalue))
  (bind _metallic (Variables|CaptureSettings|GetMetallic))
  (bind _returnvalue_8 (Utilities|String|SelectString ", Metallic" "" _metallic))
  (bind _normal (Variables|CaptureSettings|GetNormal))
  (bind _returnvalue_9 (Utilities|String|SelectString ", WorldNormal" "" _normal))
  (bind _returnvalue_10 (Utilities|String|Append _returnvalue_7 _returnvalue_9))
  (bind _returnvalue_11 (Utilities|String|Append _returnvalue_10 _returnvalue_8))
  (bind _returnvalue_12 (Utilities|String|Append _returnvalue_11 _returnvalue_2))
  (bind _returnvalue_13 (Utilities|String|Append _returnvalue_12 _returnvalue_3))
  (bind _returnvalue_14 (Utilities|String|Append _returnvalue_13 _returnvalue_4))
  (bind _returnvalue_15 (Utilities|String|Append _returnvalue_14 _returnvalue_5))
  (bind _render_type (Variables|CaptureSettings|GetRender%20Type))
  (bind _scenecolor (Variables|CaptureSettings|GetSceneColor))
  (bind _render_motion_vectors (Variables|RenderFlipbook|GetRender%20Motion%20Vectors))
  (bind _returnvalue_16 (or _scenecolor _render_motion_vectors))
  (bind _render_motionvectors (Variables|Render3DImposterSprites|GetRender%20MotionVectors))
  (bind _returnvalue_17 (or _scenecolor _render_motionvectors))
  (bind _returnvalue_18 (Utilities|Select :Index _render_type :NewEnumerator10 _scenecolor :NewEnumerator0 _returnvalue_17 :NewEnumerator5 _scenecolor :NewEnumerator2 true :NewEnumerator3 true :NewEnumerator4 true :NewEnumerator6 _returnvalue_16 :NewEnumerator8 _scenecolor :NewEnumerator9 _scenecolor))
  (bind _returnvalue_19 (Utilities|String|SelectString ", HDRColor" "" _returnvalue_18))
  (bind _scenedepthworldunits (Variables|CaptureSettings|GetSceneDepthWorldUnits))
  (bind _returnvalue_20 (Utilities|String|SelectString ", SceneDepthWorldUnits" "" _scenedepthworldunits))
  (bind _subsurface_color (Variables|CaptureSettings|GetSubsurface%20Color))
  (bind _returnvalue_21 (Utilities|String|SelectString ", SubsurfaceColor" "" _subsurface_color))
  (bind _lighting_only (Variables|CaptureSettings|GetLighting%20Only))
  (bind _returnvalue_22 (Utilities|String|SelectString ", lightingonly" "" _lighting_only))
  (bind _decal_mask (Variables|CaptureSettings|GetDecal%20Mask))
  (bind _returnvalue_23 (Utilities|String|SelectString ", DecalMask" "" _decal_mask))
  (bind _returnvalue_24 (Utilities|String|Append _returnvalue_15 _returnvalue_23))
  (bind _returnvalue_25 (Utilities|String|Append _returnvalue_24 _returnvalue_22))
  (bind _returnvalue_26 (Utilities|String|Append _returnvalue_25 _returnvalue_21))
  (bind _returnvalue_27 (Utilities|String|Append _returnvalue_26 _returnvalue_20))
  (bind _returnvalue_28 (Utilities|String|Append _returnvalue_27 _returnvalue_19))
  (bind _returnvalue_29 (Utilities|String|Append "r.BufferVisualizationOverviewTargets " _returnvalue_28))
  (Variables|Debug|SetBufferCommands _returnvalue_29)
  (Variables|Default|SetShotCommand (Utilities|String|Append "highresshot " (Utilities|String|ToString(Integer) (Math|Integer|Clamp(Integer) (Variables|CaptureSettings|GetResolutionMultiplier) 1 (Math|Integer|SelectInt 32 4 (<= (Variables|CaptureSettings|GetViewport%20size%20%28see%20Tooltip%29) 512))))))
  (Development|ExecuteConsoleCommand _returnvalue_29)
  (return))

(fn SetupMPC
  (bind (_x _y _z) (Math|Rotator|BreakRotIntoAxes (Transformation|GetActorRotation)))
  (Rendering|Material|SetVectorParameterValue "/Engine/ArtTools/RenderToTexture/Materials/MPC_R2T.MPC_R2T" "CameraForwardAxis" (Math|Conversions|ToLinearColor(Vector) _z) :!class /Script/Engine.KismetMaterialLibrary)
  (Rendering|Material|SetVectorParameterValue "/Engine/ArtTools/RenderToTexture/Materials/MPC_R2T.MPC_R2T" "BlueprintLocation" (Math|Conversions|ToLinearColor(Vector) (Transformation|GetActorLocation)) :!class /Script/Engine.KismetMaterialLibrary)
  (Rendering|Material|SetScalarParameterValue "/Engine/ArtTools/RenderToTexture/Materials/MPC_R2T.MPC_R2T" "OrthoFOV" (Variables|Render3DImposterSprites|GetOrtho%20FOV) :!class /Script/Engine.KismetMaterialLibrary))

(fn SetupCollision
  (params (Visible :type Boolean))
  (bind _sheetsize (Variables|Debug|GetSheetSize))
  (bind _returnvalue (/ _sheetsize 2))
  (bind _sheet_height (Variables|CaptureSettings|GetSheet%20height))
  (bind _returnvalue_2 (- _sheet_height 10))
  (bind _returnvalue_3 (Transformation|GetActorTransform))
  (bind _returnvalue_4 (Transformation|GetActorTransform))
  (bind _returnvalue_5 (Transformation|GetActorLocation))
  (bind _returnvalue_6 (Math|Vector|MakeVector 0.0 0.0 _returnvalue_2))
  (bind _returnvalue_7 (* _sheetsize 0.15))
  (bind _returnvalue_8 (+ _returnvalue _returnvalue_7))
  (bind _returnvalue_9 (+ _returnvalue_8 200))
  (bind _sheet_height_2 (Variables|CaptureSettings|GetSheet%20height))
  (bind _returnvalue_10 (+ _sheet_height_2 1000))
  (bind _returnvalue_11 (+ _returnvalue_8 100))
  (bind _returnvalue_12 (* _returnvalue_11 -1))
  (bind _returnvalue_13 (+ _returnvalue_5 _returnvalue_6))
  (bind _box3 (Variables|Default|GetBox3))
  (bind _box2 (Variables|Default|GetBox2))
  (bind _bottomcollision (Variables|Default|GetBottomCollision))
  (bind _bottomcollision_largemeshesonly (Variables|Default|GetBottomCollision-LargeMeshesOnly))
  (Rendering|SetVisibility (Variables|Default|GetBox2) Visible false)
  ;[1 1 1] Bottom
  (Transformation|SetWorldLocation _bottomcollision _returnvalue_13 false false)
  (Components|Box|SetBoxExtent _bottomcollision (Math|Vector|MakeVector _returnvalue_9 _returnvalue_9 10) true)
  (Transformation|SetWorldLocation _bottomcollision_largemeshesonly (+ _returnvalue_13 (Math|Vector|MakeVector 0.0 0.0 (* (Variables|PhysicsGround|GetLarge%20Mesh%20Sink%20Height) -1))) false false)
  (Components|Box|SetBoxExtent _bottomcollision_largemeshesonly (Math|Vector|MakeVector _returnvalue_8 _returnvalue_8 10) true)
  ;[1 1 1] X Walls
  (Components|Box|SetBoxExtent _box2 (Math|Vector|MakeVector 100 _returnvalue_9 1000) true)
  (Transformation|SetWorldLocation _box2 (Math|Transform|TransformLocation _returnvalue_4 (Math|Vector|MakeVector _returnvalue_12 0.0 _returnvalue_10)) false false)
  (Transformation|SetWorldLocation (Variables|Default|GetBox5) (Math|Transform|TransformLocation _returnvalue_4 (Math|Vector|MakeVector _returnvalue_11 0.0 _returnvalue_10)) false false)
  ;[1 1 1] Y Walls
  (Components|Box|SetBoxExtent _box3 (Math|Vector|MakeVector _returnvalue_8 100 1000) true)
  (Transformation|SetWorldLocation _box3 (Math|Transform|TransformLocation _returnvalue_3 (Math|Vector|MakeVector 0.0 _returnvalue_12 _returnvalue_10)) false false)
  (Transformation|SetWorldLocation (Variables|Default|GetBox4) (Math|Transform|TransformLocation _returnvalue_3 (Math|Vector|MakeVector 0.0 _returnvalue_11 _returnvalue_10)) false false)
  (return))

(fn ConstructionScript
  (bind _sheetsize (Variables|Debug|GetSheetSize))
  (bind _returnvalue (/ _sheetsize 2))
  (bind _returnvalue_2 (Math|Trig|Tan(Degrees) 0.5))
  (bind _returnvalue_3 (/ _returnvalue _returnvalue_2))
  (bind _returnvalue_4 (* "0, 0,1.000000" _returnvalue_3))
  (bind _viewport_size__see_tooltip (Variables|CaptureSettings|GetViewport%20size%20%28see%20Tooltip%29))
  (bind _returnvalue_5 (>= _viewport_size__see_tooltip 2048))
  (bind _returnvalue_6 (Math|Integer|SelectInt 2048 1024 _returnvalue_5))
  (bind _returnvalue_7 (>= _viewport_size__see_tooltip 1024))
  (bind _returnvalue_8 (Math|Integer|SelectInt _returnvalue_6 512 _returnvalue_7))
  (bind _returnvalue_9 (Transformation|GetActorLocation))
  (bind _returnvalue_10 (+ _returnvalue_9 _returnvalue_4))
  (bind _camera1 (Variables|Default|GetCamera1))
  (bind _topoutline (Variables|Default|GetTopOutline))
  (bind _sheetsize_2 (Variables|Debug|GetSheetSize))
  (bind _previewmesh (Variables|Default|GetPreviewMesh))
  (Utilities|FlowControl|Sequence
    (:then_0
      ;[1 1 1] Blueprint Set Up
      (Variables|Debug|SetResCommand (Utilities|String|BuildString(Integer) "r.setres " (Utilities|String|Append (Utilities|String|ToString(Integer) _returnvalue_8) "x") _returnvalue_8))
      (Variables|CaptureSettings|SetViewport%20size%20%28see%20Tooltip%29 _returnvalue_8))
    (:then_1
      ;[1 1 1] Background Sheet
      (Transformation|SetWorldLocation (Variables|Default|GetPlaneMesh) (+ (Transformation|GetActorLocation) (* (Math|Transform|TransformDirection (Transformation|GetActorTransform) (Math|Vector|MakeVector 0.0 0.0 1)) (Variables|CaptureSettings|GetSheet%20height))) false false)
      (Transformation|SetWorldScale3D (Variables|Default|GetPlaneMesh) (Math|Conversions|ToVector(Float) (Variables|Debug|GetSheetSize)))
      (bind _returnvalue_11 (Rendering|Material|CreateDynamicMaterialInstance (Variables|Default|GetPlaneMesh) 0 "/Engine/ArtTools/RenderToTexture/Materials/M_BackgroundCard_01.M_BackgroundCard_01" "None" :!class /Script/Engine.PrimitiveComponent))
      (Variables|Default|SetBackgroundSheetMID _returnvalue_11)
      (Rendering|Material|SetVectorParameterValue _returnvalue_11 "Color" (Variables|CaptureSettings|GetBackground%20Sheet%20Color) :!class /Script/Engine.MaterialInstanceDynamic))
    (:then_2
      ;[1 1 1] Preview Sheets
      (Transformation|SetWorldLocation (Variables|Default|GetTilingPreviewCapture) _returnvalue_10 false false)
      (Transformation|SetWorldLocation (Variables|Default|GetSceneCapture%20Motion%20Position%201) (+ _returnvalue_10 (Math|Vector|MakeVector 0.0 (Math|Conversions|ToFloat(Integer) (Math|Float|Floor (* (Variables|Debug|GetSheetSize) -1.25))))) false false)
      (Transformation|SetWorldScale3D _previewmesh (Math|Conversions|ToVector(Float) (Variables|Debug|GetSheetSize)))
      (Transformation|SetWorldLocation _previewmesh (Math|Transform|TransformLocation (Transformation|GetActorTransform) (Variables|Debug|GetPreviewLocation)) false false)
      ;[1 1 1] Top Outline, useful for checking size
      (Components|Box|SetBoxExtent _topoutline (Math|Vector|MakeVector (/ (Variables|Debug|GetSheetSize) 2) (/ (Variables|Debug|GetSheetSize) 2) 10) true)
      (Transformation|AddLocalOffset _topoutline (Math|Vector|MakeVector 0.0 0.0 (- 0.0 (* (Variables|CaptureSettings|GetSheet%20height) 10))) false false))
    (:then_3
      ;[1 1 1] Camera Setup
      (Variables|/Script/Engine.CameraComponent|SetOrthoWidth _sheetsize_2 _camera1)
      (Transformation|SetRelativeLocation _camera1 (Math|Vector|MakeVector 0.0 0.0 (* _sheetsize_2 2)) false false) ; Move camera up by sheet size
      (CallFunction|SetBufferCommands)
      (CallFunction|SetUpMPC)
      (switch Utilities|FlowControl|Switch|SwitchonRenderToTexture_Enum (Variables|CaptureSettings|GetRender%20Type)
        (:NewEnumerator10
          (CallFunction|RenderTilingMaterial))
        (:NewEnumerator0
          _)
        (:NewEnumerator5
          (CallFunction|RenderUnwrappedMeshtoTextures))
        (:NewEnumerator2
          (CallFunction|RenderDepthMap))
        (:NewEnumerator3
          (CallFunction|RenderLightmap))
        (:NewEnumerator4
          (CallFunction|RenderLIghtmap2-sided))
        (:NewEnumerator6
          _)
        (:NewEnumerator8
          (CallFunction|RenderRandomTilingPhysicsDrop))
        (:NewEnumerator9
          (CallFunction|RenderTilingMaterialFromMeshes)))
      (CallFunction|FlipBook))))
