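; Capture pawn for the RenderToTexture art tool that exports textures from a scene set up by the RenderToTexture_LevelBP generator.
; On BeginPlay it finds the generator, matches its orthographic camera to the generator's camera and sheet size, and enables buffer-visualization frame dumps, optionally HDR.
; Render events run the generator's console screenshot commands, then switch on render type to swap in normalized-depth or heightmap materials.
; The lightmap event toggles two-sided unwrap material opacity between shots.
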(event Custom|debugdepth
  (bind _generator (Variables|Default|GetGenerator))
  (bind _impostermeshcomponents (Variables|/Engine/ArtTools/RenderToTexture/Blueprints/RenderToTexture_LevelBP.RenderToTexture_LevelBP_C|GetImposterMeshComponents _generator))
  (bind _output (Utilities|Array|Get(acopy) _impostermeshcomponents))
  (bind _returnvalue (Rendering|Material|GetNumMaterials _output))
  ;[1 1 1] Imposter Depth Mask Setup
  (if (Variables|/Engine/ArtTools/RenderToTexture/Blueprints/RenderToTexture_LevelBP.RenderToTexture_LevelBP_C|GetOutput%20Depth%20Maps _generator)
    (for _index (range (+ _returnvalue 1))
      (bind _returnvalue_2 (Rendering|Material|CreateDynamicMaterialInstance "/Engine/ArtTools/RenderToTexture/Materials/M_RenderNormalizedDepth_01_Inst.M_RenderNormalizedDepth_01_Inst" "None" "None" :!class /Script/Engine.KismetMaterialLibrary))
      (Utilities|Array|Add (Variables|Default|GetMIDArray) _returnvalue_2)
      (Rendering|Material|SetTextureParameterValue _returnvalue_2 "Opacity Mask" (select (> _index (Utilities|Array|LastIndex (Variables|/Engine/ArtTools/RenderToTexture/Blueprints/RenderToTexture_LevelBP.RenderToTexture_LevelBP_C|GetOpacity%20Mask%20Textures _generator))) "/Engine/ArtTools/RenderToTexture/Textures/T_EV_BlankWhite_01.T_EV_BlankWhite_01" (Utilities|Array|Get(acopy) (Variables|/Engine/ArtTools/RenderToTexture/Blueprints/RenderToTexture_LevelBP.RenderToTexture_LevelBP_C|GetOpacity%20Mask%20Textures _generator) _index)))
      (Rendering|Material|SetVectorParameterValue _returnvalue_2 "Mask Channel" (Math|Color|SelectColor (Math|Color|MakeColor 0.0) (Utilities|Array|Get(acopy) (Variables|/Engine/ArtTools/RenderToTexture/Blueprints/RenderToTexture_LevelBP.RenderToTexture_LevelBP_C|GetOpacity%20Mask%20Channels _generator) _index) (> _index (Utilities|Array|LastIndex (Variables|/Engine/ArtTools/RenderToTexture/Blueprints/RenderToTexture_LevelBP.RenderToTexture_LevelBP_C|GetOpacity%20Mask%20Channels _generator)))) :!class /Script/Engine.MaterialInstanceDynamic))
    (for _array_element _impostermeshcomponents
      (for _index_2 (range (+ _returnvalue 1))
        (Rendering|Material|SetMaterial _array_element _index_2 (Utilities|Array|Get(acopy) (Variables|Default|GetMIDArray) _index_2))))
    (Rendering|Material|SetScalarParameterValue (Variables|/Engine/ArtTools/RenderToTexture/Blueprints/RenderToTexture_LevelBP.RenderToTexture_LevelBP_C|GetBackgroundSheetMID _generator) "Grey" 1 :!class /Script/Engine.MaterialInstanceDynamic)
    ;[1 1 1] Depth only screenshot
    (Development|ExecuteConsoleCommand "r.BufferVisualizationOverviewTargets BaseColor")
    (Development|ExecuteConsoleCommand "viewmode unlit")
    (Development|ExecuteConsoleCommand (Variables|/Engine/ArtTools/RenderToTexture/Blueprints/RenderToTexture_LevelBP.RenderToTexture_LevelBP_C|GetShotCommand (Variables|Default|GetGenerator)))))

(event Custom|rendertextures
  (bind _generator (Variables|Default|GetGenerator))
  (bind _impostermeshcomponents (Variables|/Engine/ArtTools/RenderToTexture/Blueprints/RenderToTexture_LevelBP.RenderToTexture_LevelBP_C|GetImposterMeshComponents _generator))
  (bind _output (Utilities|Array|Get(acopy) _impostermeshcomponents))
  (bind _returnvalue (Rendering|Material|GetNumMaterials _output))
  (bind _generator_2 (Variables|Default|GetGenerator))
  (bind _generator_3 (Variables|Default|GetGenerator))
  (bind _displacement_min (Variables|/Engine/ArtTools/RenderToTexture/Blueprints/RenderToTexture_LevelBP.RenderToTexture_LevelBP_C|GetDisplacement%20Min _generator_2))
  (bind _generator_4 (Variables|Default|GetGenerator))
  (Development|ExecuteConsoleCommand (Variables|/Engine/ArtTools/RenderToTexture/Blueprints/RenderToTexture_LevelBP.RenderToTexture_LevelBP_C|GetBufferCommands (Variables|Default|GetGenerator)))
  ;[1 1 1] Phys Ground Meshes
  (Development|ExecuteConsoleCommand (Variables|/Engine/ArtTools/RenderToTexture/Blueprints/RenderToTexture_LevelBP.RenderToTexture_LevelBP_C|GetShotCommand _generator_3))
  (switch Utilities|FlowControl|Switch|SwitchonRenderToTexture_Enum (Variables|/Engine/ArtTools/RenderToTexture/Blueprints/RenderToTexture_LevelBP.RenderToTexture_LevelBP_C|GetRender%20Type _generator_3)
    (:NewEnumerator0
      (Utilities|FlowControl|Delay 1)
      ;[1 1 1] Imposter Depth Mask Setup
      (if (Variables|/Engine/ArtTools/RenderToTexture/Blueprints/RenderToTexture_LevelBP.RenderToTexture_LevelBP_C|GetOutput%20Depth%20Maps _generator)
        (for _index (range (+ _returnvalue 1))
          (bind _returnvalue_2 (Rendering|Material|CreateDynamicMaterialInstance "/Engine/ArtTools/RenderToTexture/Materials/M_RenderNormalizedDepth_01_Inst.M_RenderNormalizedDepth_01_Inst" "None" "None" :!class /Script/Engine.KismetMaterialLibrary))
          (Utilities|Array|Add (Variables|Default|GetMIDArray) _returnvalue_2)
          (Rendering|Material|SetTextureParameterValue _returnvalue_2 "Opacity Mask" (select (> _index (Utilities|Array|LastIndex (Variables|/Engine/ArtTools/RenderToTexture/Blueprints/RenderToTexture_LevelBP.RenderToTexture_LevelBP_C|GetOpacity%20Mask%20Textures _generator))) "/Engine/ArtTools/RenderToTexture/Textures/T_EV_BlankWhite_01.T_EV_BlankWhite_01" (Utilities|Array|Get(acopy) (Variables|/Engine/ArtTools/RenderToTexture/Blueprints/RenderToTexture_LevelBP.RenderToTexture_LevelBP_C|GetOpacity%20Mask%20Textures _generator) _index)))
          (Rendering|Material|SetVectorParameterValue _returnvalue_2 "Mask Channel" (Math|Color|SelectColor (Math|Color|MakeColor 0.0) (Utilities|Array|Get(acopy) (Variables|/Engine/ArtTools/RenderToTexture/Blueprints/RenderToTexture_LevelBP.RenderToTexture_LevelBP_C|GetOpacity%20Mask%20Channels _generator) _index) (> _index (Utilities|Array|LastIndex (Variables|/Engine/ArtTools/RenderToTexture/Blueprints/RenderToTexture_LevelBP.RenderToTexture_LevelBP_C|GetOpacity%20Mask%20Channels _generator)))) :!class /Script/Engine.MaterialInstanceDynamic))
        (for _array_element _impostermeshcomponents
          (for _index_2 (range (+ _returnvalue 1))
            (Rendering|Material|SetMaterial _array_element _index_2 (Utilities|Array|Get(acopy) (Variables|Default|GetMIDArray) _index_2))))
        (Rendering|Material|SetScalarParameterValue (Variables|/Engine/ArtTools/RenderToTexture/Blueprints/RenderToTexture_LevelBP.RenderToTexture_LevelBP_C|GetBackgroundSheetMID _generator) "Grey" 1 :!class /Script/Engine.MaterialInstanceDynamic)))
    (:NewEnumerator8
      (Utilities|FlowControl|Delay 1)
      (bind _returnvalue_3 (Rendering|Material|CreateDynamicMaterialInstance "/Engine/ArtTools/RenderToTexture/Materials/M_RenderCustomHeightmap_01.M_RenderCustomHeightmap_01" "None" "None" :!class /Script/Engine.KismetMaterialLibrary))
      (Rendering|Material|SetScalarParameterValue _returnvalue_3 "Min" _displacement_min :!class /Script/Engine.MaterialInstanceDynamic)
      (Rendering|Material|SetScalarParameterValue _returnvalue_3 "Length" (- (Variables|/Engine/ArtTools/RenderToTexture/Blueprints/RenderToTexture_LevelBP.RenderToTexture_LevelBP_C|GetDisplacement%20Max _generator_2) _displacement_min) :!class /Script/Engine.MaterialInstanceDynamic)
      (for _array_element_2 (Variables|/Engine/ArtTools/RenderToTexture/Blueprints/RenderToTexture_LevelBP.RenderToTexture_LevelBP_C|GetSavedMeshComponents _generator_2)
        (Rendering|Material|SetMaterial _array_element_2 0 _returnvalue_3))
      (Rendering|Material|SetMaterial (Variables|/Engine/ArtTools/RenderToTexture/Blueprints/RenderToTexture_LevelBP.RenderToTexture_LevelBP_C|GetPlaneMesh _generator_2) 0 "/Engine/ArtTools/RenderToTexture/Materials/M_BackgroundCard_01.M_BackgroundCard_01")
      (bind _outactors (Actor|GetAllActorsOfClass "/Script/Engine.StaticMeshActor"))
      (for _array_element_3 _outactors
        (Rendering|Material|SetMaterial (Variables|/Script/Engine.StaticMeshActor|GetStaticMeshComponent _array_element_3) 0 _returnvalue_3)
        ;[1 1 1] Depth only screenshot
        (Development|ExecuteConsoleCommand "r.BufferVisualizationOverviewTargets BaseColor")
        (Development|ExecuteConsoleCommand "viewmode unlit")
        (Development|ExecuteConsoleCommand (Variables|/Engine/ArtTools/RenderToTexture/Blueprints/RenderToTexture_LevelBP.RenderToTexture_LevelBP_C|GetShotCommand (Variables|Default|GetGenerator)))))
    (:NewEnumerator9
      (Utilities|FlowControl|Delay 1)
      (bind _returnvalue_4 (Rendering|Material|CreateDynamicMaterialInstance "/Engine/ArtTools/RenderToTexture/Materials/M_RenderCustomHeightmap_01.M_RenderCustomHeightmap_01" "None" "None" :!class /Script/Engine.KismetMaterialLibrary))
      (Rendering|Material|SetScalarParameterValue _returnvalue_4 "Min" (Variables|/Engine/ArtTools/RenderToTexture/Blueprints/RenderToTexture_LevelBP.RenderToTexture_LevelBP_C|GetDisplacementMin _generator_4) :!class /Script/Engine.MaterialInstanceDynamic)
      (Rendering|Material|SetScalarParameterValue _returnvalue_4 "Max" (Variables|/Engine/ArtTools/RenderToTexture/Blueprints/RenderToTexture_LevelBP.RenderToTexture_LevelBP_C|GetDisplacementMax _generator_4) :!class /Script/Engine.MaterialInstanceDynamic)
      (for _array_element_4 (Variables|/Engine/ArtTools/RenderToTexture/Blueprints/RenderToTexture_LevelBP.RenderToTexture_LevelBP_C|GetCopied%20Mesh%20Array _generator_4)
        (Rendering|Material|SetMaterial _array_element_4 0 _returnvalue_4))
      (Rendering|Material|SetMaterial (Variables|/Engine/ArtTools/RenderToTexture/Blueprints/RenderToTexture_LevelBP.RenderToTexture_LevelBP_C|GetPlaneMesh _generator_4) 0 "/Engine/ArtTools/RenderToTexture/Materials/M_BackgroundCard_01.M_BackgroundCard_01"))))

(event Custom|renderdepth
  (bind _generator (Variables|Default|GetGenerator))
  (bind _impostermeshcomponents (Variables|/Engine/ArtTools/RenderToTexture/Blueprints/RenderToTexture_LevelBP.RenderToTexture_LevelBP_C|GetImposterMeshComponents _generator))
  (bind _output (Utilities|Array|Get(acopy) _impostermeshcomponents))
  (bind _returnvalue (Rendering|Material|GetNumMaterials _output))
  (bind _generator_2 (Variables|Default|GetGenerator))
  (bind _generator_3 (Variables|Default|GetGenerator))
  (bind _displacement_min (Variables|/Engine/ArtTools/RenderToTexture/Blueprints/RenderToTexture_LevelBP.RenderToTexture_LevelBP_C|GetDisplacement%20Min _generator_2))
  (bind _generator_4 (Variables|Default|GetGenerator))
  (Utilities|FlowControl|Sequence
    (:then_0
      (Development|ExecuteConsoleCommand "viewmode unlit"))
    (:then_1
      (Development|ExecuteConsoleCommand (Variables|/Engine/ArtTools/RenderToTexture/Blueprints/RenderToTexture_LevelBP.RenderToTexture_LevelBP_C|GetBufferCommands (Variables|Default|GetGenerator)))
      ;[1 1 1] Phys Ground Meshes
      (Development|ExecuteConsoleCommand (Variables|/Engine/ArtTools/RenderToTexture/Blueprints/RenderToTexture_LevelBP.RenderToTexture_LevelBP_C|GetShotCommand _generator_3))
      (switch Utilities|FlowControl|Switch|SwitchonRenderToTexture_Enum (Variables|/Engine/ArtTools/RenderToTexture/Blueprints/RenderToTexture_LevelBP.RenderToTexture_LevelBP_C|GetRender%20Type _generator_3)
        (:NewEnumerator0
          (Utilities|FlowControl|Delay 1)
          ;[1 1 1] Imposter Depth Mask Setup
          (if (Variables|/Engine/ArtTools/RenderToTexture/Blueprints/RenderToTexture_LevelBP.RenderToTexture_LevelBP_C|GetOutput%20Depth%20Maps _generator)
            (for _index (range (+ _returnvalue 1))
              (bind _returnvalue_2 (Rendering|Material|CreateDynamicMaterialInstance "/Engine/ArtTools/RenderToTexture/Materials/M_RenderNormalizedDepth_01_Inst.M_RenderNormalizedDepth_01_Inst" "None" "None" :!class /Script/Engine.KismetMaterialLibrary))
              (Utilities|Array|Add (Variables|Default|GetMIDArray) _returnvalue_2)
              (Rendering|Material|SetTextureParameterValue _returnvalue_2 "Opacity Mask" (select (> _index (Utilities|Array|LastIndex (Variables|/Engine/ArtTools/RenderToTexture/Blueprints/RenderToTexture_LevelBP.RenderToTexture_LevelBP_C|GetOpacity%20Mask%20Textures _generator))) "/Engine/ArtTools/RenderToTexture/Textures/T_EV_BlankWhite_01.T_EV_BlankWhite_01" (Utilities|Array|Get(acopy) (Variables|/Engine/ArtTools/RenderToTexture/Blueprints/RenderToTexture_LevelBP.RenderToTexture_LevelBP_C|GetOpacity%20Mask%20Textures _generator) _index)))
              (Rendering|Material|SetVectorParameterValue _returnvalue_2 "Mask Channel" (Math|Color|SelectColor (Math|Color|MakeColor 0.0) (Utilities|Array|Get(acopy) (Variables|/Engine/ArtTools/RenderToTexture/Blueprints/RenderToTexture_LevelBP.RenderToTexture_LevelBP_C|GetOpacity%20Mask%20Channels _generator) _index) (> _index (Utilities|Array|LastIndex (Variables|/Engine/ArtTools/RenderToTexture/Blueprints/RenderToTexture_LevelBP.RenderToTexture_LevelBP_C|GetOpacity%20Mask%20Channels _generator)))) :!class /Script/Engine.MaterialInstanceDynamic))
            (for _array_element _impostermeshcomponents
              (for _index_2 (range (+ _returnvalue 1))
                (Rendering|Material|SetMaterial _array_element _index_2 (Utilities|Array|Get(acopy) (Variables|Default|GetMIDArray) _index_2))))
            (Rendering|Material|SetScalarParameterValue (Variables|/Engine/ArtTools/RenderToTexture/Blueprints/RenderToTexture_LevelBP.RenderToTexture_LevelBP_C|GetBackgroundSheetMID _generator) "Grey" 1 :!class /Script/Engine.MaterialInstanceDynamic)))
        (:NewEnumerator8
          (Utilities|FlowControl|Delay 1)
          (bind _returnvalue_3 (Rendering|Material|CreateDynamicMaterialInstance "/Engine/ArtTools/RenderToTexture/Materials/M_RenderCustomHeightmap_01.M_RenderCustomHeightmap_01" "None" "None" :!class /Script/Engine.KismetMaterialLibrary))
          (Rendering|Material|SetScalarParameterValue _returnvalue_3 "Min" _displacement_min :!class /Script/Engine.MaterialInstanceDynamic)
          (Rendering|Material|SetScalarParameterValue _returnvalue_3 "Length" (- (Variables|/Engine/ArtTools/RenderToTexture/Blueprints/RenderToTexture_LevelBP.RenderToTexture_LevelBP_C|GetDisplacement%20Max _generator_2) _displacement_min) :!class /Script/Engine.MaterialInstanceDynamic)
          (for _array_element_2 (Variables|/Engine/ArtTools/RenderToTexture/Blueprints/RenderToTexture_LevelBP.RenderToTexture_LevelBP_C|GetSavedMeshComponents _generator_2)
            (Rendering|Material|SetMaterial _array_element_2 0 _returnvalue_3))
          (Rendering|Material|SetMaterial (Variables|/Engine/ArtTools/RenderToTexture/Blueprints/RenderToTexture_LevelBP.RenderToTexture_LevelBP_C|GetPlaneMesh _generator_2) 0 "/Engine/ArtTools/RenderToTexture/Materials/M_BackgroundCard_01.M_BackgroundCard_01")
          (bind _outactors (Actor|GetAllActorsOfClass "/Script/Engine.StaticMeshActor"))
          (for _array_element_3 _outactors
            (Rendering|Material|SetMaterial (Variables|/Script/Engine.StaticMeshActor|GetStaticMeshComponent _array_element_3) 0 _returnvalue_3)
            ;[1 1 1] Depth only screenshot
            (Development|ExecuteConsoleCommand "r.BufferVisualizationOverviewTargets BaseColor")
            (Development|ExecuteConsoleCommand "viewmode unlit")
            (Development|ExecuteConsoleCommand (Variables|/Engine/ArtTools/RenderToTexture/Blueprints/RenderToTexture_LevelBP.RenderToTexture_LevelBP_C|GetShotCommand (Variables|Default|GetGenerator)))))
        (:NewEnumerator9
          (Utilities|FlowControl|Delay 1)
          (bind _returnvalue_4 (Rendering|Material|CreateDynamicMaterialInstance "/Engine/ArtTools/RenderToTexture/Materials/M_RenderCustomHeightmap_01.M_RenderCustomHeightmap_01" "None" "None" :!class /Script/Engine.KismetMaterialLibrary))
          (Rendering|Material|SetScalarParameterValue _returnvalue_4 "Min" (Variables|/Engine/ArtTools/RenderToTexture/Blueprints/RenderToTexture_LevelBP.RenderToTexture_LevelBP_C|GetDisplacementMin _generator_4) :!class /Script/Engine.MaterialInstanceDynamic)
          (Rendering|Material|SetScalarParameterValue _returnvalue_4 "Max" (Variables|/Engine/ArtTools/RenderToTexture/Blueprints/RenderToTexture_LevelBP.RenderToTexture_LevelBP_C|GetDisplacementMax _generator_4) :!class /Script/Engine.MaterialInstanceDynamic)
          (for _array_element_4 (Variables|/Engine/ArtTools/RenderToTexture/Blueprints/RenderToTexture_LevelBP.RenderToTexture_LevelBP_C|GetCopied%20Mesh%20Array _generator_4)
            (Rendering|Material|SetMaterial _array_element_4 0 _returnvalue_4))
          (Rendering|Material|SetMaterial (Variables|/Engine/ArtTools/RenderToTexture/Blueprints/RenderToTexture_LevelBP.RenderToTexture_LevelBP_C|GetPlaneMesh _generator_4) 0 "/Engine/ArtTools/RenderToTexture/Materials/M_BackgroundCard_01.M_BackgroundCard_01"))))))

(event Custom|renderimposter
  (bind _generator (Variables|Default|GetGenerator))
  (bind _impostermeshcomponents (Variables|/Engine/ArtTools/RenderToTexture/Blueprints/RenderToTexture_LevelBP.RenderToTexture_LevelBP_C|GetImposterMeshComponents _generator))
  (bind _output (Utilities|Array|Get(acopy) _impostermeshcomponents))
  (bind _returnvalue (Rendering|Material|GetNumMaterials _output))
  (bind _generator_2 (Variables|Default|GetGenerator))
  (bind _generator_3 (Variables|Default|GetGenerator))
  (bind _displacement_min (Variables|/Engine/ArtTools/RenderToTexture/Blueprints/RenderToTexture_LevelBP.RenderToTexture_LevelBP_C|GetDisplacement%20Min _generator_2))
  (bind _generator_4 (Variables|Default|GetGenerator))
  (Development|ExecuteConsoleCommand (Variables|/Engine/ArtTools/RenderToTexture/Blueprints/RenderToTexture_LevelBP.RenderToTexture_LevelBP_C|GetBufferCommands (Variables|Default|GetGenerator)))
  ;[1 1 1] Phys Ground Meshes
  (Development|ExecuteConsoleCommand (Variables|/Engine/ArtTools/RenderToTexture/Blueprints/RenderToTexture_LevelBP.RenderToTexture_LevelBP_C|GetShotCommand _generator_3))
  (switch Utilities|FlowControl|Switch|SwitchonRenderToTexture_Enum (Variables|/Engine/ArtTools/RenderToTexture/Blueprints/RenderToTexture_LevelBP.RenderToTexture_LevelBP_C|GetRender%20Type _generator_3)
    (:NewEnumerator0
      (Utilities|FlowControl|Delay 1)
      ;[1 1 1] Imposter Depth Mask Setup
      (if (Variables|/Engine/ArtTools/RenderToTexture/Blueprints/RenderToTexture_LevelBP.RenderToTexture_LevelBP_C|GetOutput%20Depth%20Maps _generator)
        (for _index (range (+ _returnvalue 1))
          (bind _returnvalue_2 (Rendering|Material|CreateDynamicMaterialInstance "/Engine/ArtTools/RenderToTexture/Materials/M_RenderNormalizedDepth_01_Inst.M_RenderNormalizedDepth_01_Inst" "None" "None" :!class /Script/Engine.KismetMaterialLibrary))
          (Utilities|Array|Add (Variables|Default|GetMIDArray) _returnvalue_2)
          (Rendering|Material|SetTextureParameterValue _returnvalue_2 "Opacity Mask" (select (> _index (Utilities|Array|LastIndex (Variables|/Engine/ArtTools/RenderToTexture/Blueprints/RenderToTexture_LevelBP.RenderToTexture_LevelBP_C|GetOpacity%20Mask%20Textures _generator))) "/Engine/ArtTools/RenderToTexture/Textures/T_EV_BlankWhite_01.T_EV_BlankWhite_01" (Utilities|Array|Get(acopy) (Variables|/Engine/ArtTools/RenderToTexture/Blueprints/RenderToTexture_LevelBP.RenderToTexture_LevelBP_C|GetOpacity%20Mask%20Textures _generator) _index)))
          (Rendering|Material|SetVectorParameterValue _returnvalue_2 "Mask Channel" (Math|Color|SelectColor (Math|Color|MakeColor 0.0) (Utilities|Array|Get(acopy) (Variables|/Engine/ArtTools/RenderToTexture/Blueprints/RenderToTexture_LevelBP.RenderToTexture_LevelBP_C|GetOpacity%20Mask%20Channels _generator) _index) (> _index (Utilities|Array|LastIndex (Variables|/Engine/ArtTools/RenderToTexture/Blueprints/RenderToTexture_LevelBP.RenderToTexture_LevelBP_C|GetOpacity%20Mask%20Channels _generator)))) :!class /Script/Engine.MaterialInstanceDynamic))
        (for _array_element _impostermeshcomponents
          (for _index_2 (range (+ _returnvalue 1))
            (Rendering|Material|SetMaterial _array_element _index_2 (Utilities|Array|Get(acopy) (Variables|Default|GetMIDArray) _index_2))))
        (Rendering|Material|SetScalarParameterValue (Variables|/Engine/ArtTools/RenderToTexture/Blueprints/RenderToTexture_LevelBP.RenderToTexture_LevelBP_C|GetBackgroundSheetMID _generator) "Grey" 1 :!class /Script/Engine.MaterialInstanceDynamic)))
    (:NewEnumerator8
      (Utilities|FlowControl|Delay 1)
      (bind _returnvalue_3 (Rendering|Material|CreateDynamicMaterialInstance "/Engine/ArtTools/RenderToTexture/Materials/M_RenderCustomHeightmap_01.M_RenderCustomHeightmap_01" "None" "None" :!class /Script/Engine.KismetMaterialLibrary))
      (Rendering|Material|SetScalarParameterValue _returnvalue_3 "Min" _displacement_min :!class /Script/Engine.MaterialInstanceDynamic)
      (Rendering|Material|SetScalarParameterValue _returnvalue_3 "Length" (- (Variables|/Engine/ArtTools/RenderToTexture/Blueprints/RenderToTexture_LevelBP.RenderToTexture_LevelBP_C|GetDisplacement%20Max _generator_2) _displacement_min) :!class /Script/Engine.MaterialInstanceDynamic)
      (for _array_element_2 (Variables|/Engine/ArtTools/RenderToTexture/Blueprints/RenderToTexture_LevelBP.RenderToTexture_LevelBP_C|GetSavedMeshComponents _generator_2)
        (Rendering|Material|SetMaterial _array_element_2 0 _returnvalue_3))
      (Rendering|Material|SetMaterial (Variables|/Engine/ArtTools/RenderToTexture/Blueprints/RenderToTexture_LevelBP.RenderToTexture_LevelBP_C|GetPlaneMesh _generator_2) 0 "/Engine/ArtTools/RenderToTexture/Materials/M_BackgroundCard_01.M_BackgroundCard_01")
      (bind _outactors (Actor|GetAllActorsOfClass "/Script/Engine.StaticMeshActor"))
      (for _array_element_3 _outactors
        (Rendering|Material|SetMaterial (Variables|/Script/Engine.StaticMeshActor|GetStaticMeshComponent _array_element_3) 0 _returnvalue_3)
        ;[1 1 1] Depth only screenshot
        (Development|ExecuteConsoleCommand "r.BufferVisualizationOverviewTargets BaseColor")
        (Development|ExecuteConsoleCommand "viewmode unlit")
        (Development|ExecuteConsoleCommand (Variables|/Engine/ArtTools/RenderToTexture/Blueprints/RenderToTexture_LevelBP.RenderToTexture_LevelBP_C|GetShotCommand (Variables|Default|GetGenerator)))))
    (:NewEnumerator9
      (Utilities|FlowControl|Delay 1)
      (bind _returnvalue_4 (Rendering|Material|CreateDynamicMaterialInstance "/Engine/ArtTools/RenderToTexture/Materials/M_RenderCustomHeightmap_01.M_RenderCustomHeightmap_01" "None" "None" :!class /Script/Engine.KismetMaterialLibrary))
      (Rendering|Material|SetScalarParameterValue _returnvalue_4 "Min" (Variables|/Engine/ArtTools/RenderToTexture/Blueprints/RenderToTexture_LevelBP.RenderToTexture_LevelBP_C|GetDisplacementMin _generator_4) :!class /Script/Engine.MaterialInstanceDynamic)
      (Rendering|Material|SetScalarParameterValue _returnvalue_4 "Max" (Variables|/Engine/ArtTools/RenderToTexture/Blueprints/RenderToTexture_LevelBP.RenderToTexture_LevelBP_C|GetDisplacementMax _generator_4) :!class /Script/Engine.MaterialInstanceDynamic)
      (for _array_element_4 (Variables|/Engine/ArtTools/RenderToTexture/Blueprints/RenderToTexture_LevelBP.RenderToTexture_LevelBP_C|GetCopied%20Mesh%20Array _generator_4)
        (Rendering|Material|SetMaterial _array_element_4 0 _returnvalue_4))
      (Rendering|Material|SetMaterial (Variables|/Engine/ArtTools/RenderToTexture/Blueprints/RenderToTexture_LevelBP.RenderToTexture_LevelBP_C|GetPlaneMesh _generator_4) 0 "/Engine/ArtTools/RenderToTexture/Materials/M_BackgroundCard_01.M_BackgroundCard_01"))))

(event EventBeginPlay
  (bind _pawncam (Variables|Default|GetPawncam))
  ;[1 1 1] Initial Setup
  (Utilities|FlowControl|Delay 0.2)
  (bind _outactors (Actor|GetAllActorsOfClass "/Engine/ArtTools/RenderToTexture/Blueprints/RenderToTexture_LevelBP.RenderToTexture_LevelBP_C"))
  (Variables|Default|SetGenerator (Utilities|Array|Get(acopy) _outactors))
  (Variables|/Script/Engine.CameraComponent|SetOrthoWidth (Variables|/Engine/ArtTools/RenderToTexture/Blueprints/RenderToTexture_LevelBP.RenderToTexture_LevelBP_C|GetSheetSize (Utilities|Array|Get(acopy) _outactors)) (Variables|Default|GetPawncam))
  (Transformation|SetWorldLocation _pawncam (Transformation|GetWorldLocation (Variables|/Engine/ArtTools/RenderToTexture/Blueprints/RenderToTexture_LevelBP.RenderToTexture_LevelBP_C|GetCamera1 (Utilities|Array|Get(acopy) _outactors))) false false)
  (Transformation|SetWorldRotation _pawncam (Transformation|GetWorldRotation (Variables|/Engine/ArtTools/RenderToTexture/Blueprints/RenderToTexture_LevelBP.RenderToTexture_LevelBP_C|GetCamera1 (Utilities|Array|Get(acopy) _outactors))) false false)
  (Development|ExecuteConsoleCommand "r.BufferVisualizationDumpFrames 1")
  (Development|ExecuteConsoleCommand (Variables|/Engine/ArtTools/RenderToTexture/Blueprints/RenderToTexture_LevelBP.RenderToTexture_LevelBP_C|GetBufferCommands (Variables|Default|GetGenerator)))
  (if (Variables|/Engine/ArtTools/RenderToTexture/Blueprints/RenderToTexture_LevelBP.RenderToTexture_LevelBP_C|GetExport%20images%20as%20EXR (Utilities|Array|Get(acopy) _outactors))
    (Development|ExecuteConsoleCommand "r.BufferVisualizationDumpFramesAsHDR 1")
    (else
      (Development|ExecuteConsoleCommand "r.BufferVisualizationDumpFramesAsHDR 0"))))

(event Custom|renderlightmaps
  ;[1 1 1] Lightmaps
  (for _array_element (Variables|/Engine/ArtTools/RenderToTexture/Blueprints/RenderToTexture_LevelBP.RenderToTexture_LevelBP_C|GetUnwrap2sidedMIDListB (Utilities|Array|Get(acopy) (stale EventBeginPlay _outactors)))
    (Rendering|Material|SetScalarParameterValue _array_element "Opacity" 0.0 :!class /Script/Engine.MaterialInstanceDynamic))
  (Development|ExecuteConsoleCommand (Variables|/Engine/ArtTools/RenderToTexture/Blueprints/RenderToTexture_LevelBP.RenderToTexture_LevelBP_C|GetBufferCommands (Variables|Default|GetGenerator)))
  (Development|ExecuteConsoleCommand (Variables|/Engine/ArtTools/RenderToTexture/Blueprints/RenderToTexture_LevelBP.RenderToTexture_LevelBP_C|GetShotCommand (Variables|Default|GetGenerator)))
  (Utilities|FlowControl|Delay 0.2)
  (for _array_element_2 (Variables|/Engine/ArtTools/RenderToTexture/Blueprints/RenderToTexture_LevelBP.RenderToTexture_LevelBP_C|GetUnwrap2sidedMIDlistA (Utilities|Array|Get(acopy) (stale EventBeginPlay _outactors)))
    (Rendering|Material|SetScalarParameterValue _array_element_2 "Opacity" 0.0 :!class /Script/Engine.MaterialInstanceDynamic))
  (for _array_element_3 (Variables|/Engine/ArtTools/RenderToTexture/Blueprints/RenderToTexture_LevelBP.RenderToTexture_LevelBP_C|GetUnwrap2sidedMIDListB (Utilities|Array|Get(acopy) (stale EventBeginPlay _outactors)))
    (Rendering|Material|SetScalarParameterValue _array_element_3 "Opacity" 1 :!class /Script/Engine.MaterialInstanceDynamic))
  (Development|ExecuteConsoleCommand (Variables|/Engine/ArtTools/RenderToTexture/Blueprints/RenderToTexture_LevelBP.RenderToTexture_LevelBP_C|GetShotCommand (Variables|Default|GetGenerator)))
  (Utilities|FlowControl|Delay 0.2)
  (for _array_element_4 (Variables|/Engine/ArtTools/RenderToTexture/Blueprints/RenderToTexture_LevelBP.RenderToTexture_LevelBP_C|GetUnwrap2sidedMIDlistA (Utilities|Array|Get(acopy) (stale EventBeginPlay _outactors)))
    (Rendering|Material|SetScalarParameterValue _array_element_4 "Opacity" 1 :!class /Script/Engine.MaterialInstanceDynamic)))

(fn ConstructionScript)
