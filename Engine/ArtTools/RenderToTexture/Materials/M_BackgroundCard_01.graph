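# Unlit translucent background card material in the engine's RenderToTexture art tools, giving a flat colour adjustable toward grey.
# A Color vector parameter's RGB is multiplied by itself, squaring it to darken values below one.
# The Grey scalar parameter blends that squared colour toward a constant mid grey.
# The graph samples no textures and shows no link from the blend to a material output.
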Material /Engine/ArtTools/RenderToTexture/Materials/M_BackgroundCard_01
graph {
 properties {
  blendMode=BLEND_Translucent
  pixelDepthOffsetMode=PDOM_Legacy
  shadingModel=MSM_Unlit
  translucencyPass=MTP_BeforeDOF
  usedShadingModels=Unlit
 }
 nodes [
  n1 = VectorParameter() {parameterName=Color}
  LinearInterpolate(A=Multiply(A=@n1.RGB, B=@n1.RGB), Alpha=ScalarParameter() {parameterName=Grey}) {constB=0.5}
 ]
}
