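# Material function that unwraps a mesh into its flat UV layout for render-to-texture capture.
# UVs are recentered, negated, scaled by Size and placed at RenderLocation on the XY plane.
# Subtracting world position yields the WorldPositionOffset output, scaled by the clamped Unwrap parameter.
# A second Mask output multiplies TwoSidedSign by a front/back selector and clamps it to mask unwanted faces.
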MaterialFunction /Engine/ArtTools/RenderToTexture/MaterialFunctions/UnwrapUVsForRender
graph {
 properties {bExposeToLibrary=true, description=A function to unwrap meshes for render to texture purposes}
 nodes [
  n1 = FunctionInput(Preview=TextureCoordinate()) {bUsePreviewValueAsDefault=true, inputName=UVs, inputType=FunctionInput_Vector2, previewValue=<100 100 0 1>}
  n2 = Multiply(A=ConstantBiasScale(None=@n1) {bias=-0.5, scale=2}) {constB=-1}
  n3 = ScalarParameter() {parameterName=Size}
  n4 = Subtract(A=AppendVector(A=Multiply(A=@n2, B=@n3), B=Constant()), B=WorldPosition() {worldPositionShaderOffset=WPT_ExcludeAllShaderOffsets}.XYZ) {constA=0}
  n5 = VectorParameter() {parameterName=RenderLocation}
  n6 = ScalarParameter() {parameterName=Unwrap}
  FunctionOutput(None=Multiply(A=Add(A=@n4, B=@n5.RGB), B=Clamp(None=@n6))) {outputName=WorldPositionOffset, sortPriority=1}
  n7 = TwoSidedSign()
  n8 = FunctionInput() {
   bUsePreviewValueAsDefault=true
   description=Enter 1 or -1
   inputName=Frontface or Backface
   inputType=FunctionInput_Scalar
   previewValue=<1 0 0 0>
  }
  FunctionOutput(None=Clamp(None=Multiply(A=@n7, B=@n8))) {description=Mask Unwated Backfaces, outputName=Mask}
 ]
}
