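# Two-sided Render To Texture tool material for comparing UV texel density against local-space size.
# It divides screen derivative lengths of UV1 by those of world position, takes the larger, scales by local bounds X and gamma-corrects.
# A Checker switch picks a checker on UV1 or bounding-sphere local UVs, or a numeric readout of bounds X.
# An UnwrapUVsForRender branch handles UV-space unwrapping; no material output connections appear.
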Material /Engine/ArtTools/RenderToTexture/Materials/M_CompareUVtoLocalSpaceDensity_01
graph {
 properties {pixelDepthOffsetMode=PDOM_Legacy, twoSided=true, usedShadingModels=Default Lit}
 nodes [
  n1 = ComponentMask(None=MaterialFunctionCall!(World Position, Scale XY, Scale Z, Clamp, Transform to Local space, Custom Light Vector, Light Vector) {materialFunction=/Engine/ArtTools/RenderToTexture/MaterialFunctions/BoundingSphere_LightTransform}.UVW) {g=true, r=true}
  Abs(None=@n1)
  n2 = TextureCoordinate() {coordinateIndex=1}
  n3 = StaticSwitch(True=ComponentMask(None=@n2) {g=true}, False=ComponentMask(None=@n2) {r=true})
  Abs(None=DDX(Value=@n3))
  Abs(None=DDY(Value=@n3))
  ActorPositionWS()
  Constant()
  LightmapUVs()
  MaterialFunctionCall!(UVs=OneMinus(None=TextureCoordinate() {coordinateIndex=1}), Frontface or Backface) {materialFunction=/Engine/ArtTools/RenderToTexture/MaterialFunctions/UnwrapUVsForRender}
  Multiply(A=ScalarParameter() {defaultValue=512, parameterName=Size}, B=Constant2Vector() {g=1, r=-1})
  n4 = MaterialFunctionCall() {materialFunction=&ObjectLocalBounds}
  Multiply(
   A=@n4.Local Bounds Size
   B=MaterialFunctionCall() {materialFunction=/Engine/Functions/Engine_MaterialFunctions02/WorldPositionOffset/ObjectScale}.Scale XYZ
  )
  ObjectBounds()
  ObjectBounds()
  ObjectPositionWS()
  ObjectRadius()
  n5 = DDX(Value=@n2)
  n6 = WorldPosition() {worldPositionShaderOffset=WPT_ExcludeAllShaderOffsets}
  n7 = DDX(Value=@n6.XYZ)
  n8 = DDY(Value=@n2)
  n9 = DDY(Value=@n6.XYZ)
  n10 = Max(
   A=Divide(
    A=MaterialFunctionCall(Vector 3=@n5, Vector 2=@n5) {materialFunction=/Engine/Functions/Engine_MaterialFunctions02/Utility/VectorLength}.V2 Length
    B=MaterialFunctionCall(Vector 3=@n7) {materialFunction=&VectorLength}.V3 Length
   )
   B=Divide(
    A=MaterialFunctionCall(Vector 3=@n8, Vector 2=@n8) {materialFunction=&VectorLength}.V2 Length
    B=MaterialFunctionCall(Vector 3=@n9) {materialFunction=&VectorLength}.V3 Length
   )
  )
  n11 = ComponentMask(None=@n4.Local Bounds Size) {r=true}
  n12 = Multiply(A=@n11) {constB=2}
  Power(Base=Multiply(A=Multiply(A=@n10, B=@n12)) {constB=0.5}) {constExponent=2.2, desc=Gamma Correction}
  ScalarParameter() {defaultValue=1, parameterName=Unwrap}
  n13 = TextureCoordinate() {coordinateIndex=1}
  n14 = MaterialFunctionCall(
   Number=@n11
   UVs=MaterialFunctionCall(UVs=TextureCoordinate(), Texture Scale=Constant() {r=0.5}) {materialFunction=&ScaleUVsByCenter}.UVs
  ) {materialFunction=&DebugScalarValues}
  n15 = Constant()
  StaticSwitchParameter(
   True=MaterialFunctionCall!(UVs=StaticSwitchParameter(True=@n13, False=@n1) {parameterName=UVs}, Tiling=ScalarParameter() {defaultValue=32, parameterName=Resolution}) {materialFunction=/Engine/ArtTools/RenderToTexture/MaterialFunctions/CheckerPattern}.Result
   False=FeatureLevelSwitch(Default=@n14.Result, ES2=@n15, ES3_1=@n15, SM4_REMOVED=@n14.Result)
  ) {parameterName=Checker}
 ]
}
