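# Material function that turns a UVW input into a debug color with center lines and circles, apparently for checking bounding-sphere framing in render-to-texture work.
# The input passes through MarkRedOutside0-1Range, then thresholded If nodes black out lines where R or G is near 0.5.
# Distance from the XY center draws a black outer ring near the 0-1 edge and a dimmed inner ring.
# A final 2.2 power adjusts the output.
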MaterialFunction /Engine/ArtTools/RenderToTexture/MaterialFunctions/Boundingsphere_DebugCircles
graph {nodes [
  n1 = MaterialFunctionCall!(UVW=FunctionInput() {bUsePreviewValueAsDefault=true, inputName=UVW, previewValue=<0.5 0.5 0.5 1>}) {materialFunction=&/MarkRedOutside0-1Range}
  n2 = MaterialFunctionCall(Float3=@n1.Result) {materialFunction=&BreakOutFloat3Components}
  n3 = Constant() {r=0.5}
  n4 = Constant()
  n5 = Min(
   A=If(A=@n2.R, B=@n3, "A > B"=@n1.Result, "A == B"=@n4, "A < B"=@n1.Result) {equalsThreshold=0.01}
   B=If(A=@n2.G, B=@n3, "A > B"=@n1.Result, "A == B"=@n4, "A < B"=@n1.Result) {equalsThreshold=0.01}
  )
  n6 = ComponentMask(None=@n5) {g=true, r=true}
  n7 = Constant()
  n8 = Multiply(A=Distance(A=AppendVector(A=@n6, B=@n7), B=Constant3Vector() {constant=<rgba 0.5 0.5 0 0>})) {constB=2}
  n9 = Constant() {desc=Inner Circle, r=0.75}
  n10 = If(A=@n8, B=Constant() {desc=Outer circle near 0-1 edge, r=0.9875}, "A > B"=@n5, "A == B"=Constant(), "A < B"=@n5) {constB=1.5, equalsThreshold=0.01}
  n11 = Multiply(A=@n10) {constB=0.4, desc=Opacity of inner circle}
  FunctionOutput(None=Power(Base=If(A=@n8, B=@n9, "A > B"=@n10, "A == B"=@n11, "A < B"=@n10) {constB=1.5, equalsThreshold=0.02}) {constExponent=2.2})
 ]}
