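# Unlit debug material that, going by its name and functions, visualizes sphere bounds for the render-to-texture art tools.
# BoundingSphere_LightTransform builds UVW coordinates from XY and Z scale parameters and a custom ProjectionVector light direction, with clamping enabled.
# Those coordinates feed Boundingsphere_DebugCircles, so the projection logic lives in shared material functions.
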Material /Engine/ArtTools/RenderToTexture/Materials/M_DebugSphereBounds_01
graph {
 properties {bUsedWithStaticLighting=true, pixelDepthOffsetMode=PDOM_Legacy, shadingModel=MSM_Unlit, usedShadingModels=Unlit}
 nodes [
  n1 = ScalarParameter() {defaultValue=1, parameterName=Scale XY}
  MaterialFunctionCall!(UVW=MaterialFunctionCall!(
    World Position
    Scale XY=AppendVector(A=@n1, B=@n1)
    Scale Z=ScalarParameter() {defaultValue=0.5, parameterName=Scale Z}
    Clamp=StaticBool() {value=true}
    Transform to Local space
    Custom Light Vector=StaticBool() {value=true}
    Light Vector=VectorParameter() {parameterName=ProjectionVector, defaultValue=<rgba 0 0 1 0>}.RGB
   ) {materialFunction=/Engine/ArtTools/RenderToTexture/MaterialFunctions/BoundingSphere_LightTransform}.UVW) {materialFunction=/Engine/ArtTools/RenderToTexture/MaterialFunctions/Boundingsphere_DebugCircles}
 ]
}
