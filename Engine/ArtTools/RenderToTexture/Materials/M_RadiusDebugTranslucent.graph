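# M_RadiusDebugTranslucent is an unlit, translucent material in the engine's RenderToTexture art tools, named for debug display of a radius.
# Its graph holds only a Color vector parameter, green by default, and a low scalar constant, with no texture sampling or math.
# It is flagged for use with static lighting.
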Material /Engine/ArtTools/RenderToTexture/Materials/M_RadiusDebugTranslucent.M_RadiusDebugTransLucent
graph {
 properties {bUsedWithStaticLighting=true, blendMode=BLEND_Translucent, pixelDepthOffsetMode=PDOM_Legacy, shadingModel=MSM_Unlit, usedShadingModels=Unlit}
 nodes [
  Constant() {r=0.1}
  VectorParameter() {parameterName=Color, defaultValue=<rgba 0 1 0 1>}
 ]
}
