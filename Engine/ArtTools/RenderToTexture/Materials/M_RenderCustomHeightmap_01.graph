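# Unlit render-to-texture material that computes normalized, displacement-adjusted depth for custom heightmap capture.
# World position is transformed along the MPC_R2T camera axis around BlueprintLocation, offset by a Displacement texture, then remapped between Min and Max.
# An Initial Depth texture, sampled by world XY over Size and gamma-adjusted, is offset and compared to that depth through Ceil, giving a step mask.
# MarkRedOutside0-1Range takes the normalized depth to flag out-of-range values.
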Material /Engine/ArtTools/RenderToTexture/Materials/M_RenderCustomHeightmap_01
graph {
 properties {pixelDepthOffsetMode=PDOM_Legacy, shadingModel=MSM_Unlit, usedShadingModels=Unlit}
 nodes [
  n1 = WorldPosition()
  n2 = ScalarParameter() {defaultValue=1000, parameterName=Size}
  n3 = Divide(A=@n2)
  n4 = ComponentMask(None=Divide(A=Add(A=@n1.XYZ, B=@n3), B=@n2)) {g=true, r=true}
  n5 = Power(Base=TextureSampleParameter2D(UVs=@n4) {parameterName=Initial Depth, samplerType=SAMPLERTYPE_LinearColor, texture=/Engine/ArtTools/RenderToTexture/Textures/127grey}.R) {constExponent=2.2}
  n6 = ScalarParameter() {defaultValue=-0.01, parameterName=Offset}
  n7 = CollectionParameter() {collection=&/MPC_R2T, parameterName=CameraForwardAxis}
  n8 = CollectionParameter() {collection=&MPC_R2T, parameterName=BlueprintLocation}
  n9 = ComponentMask(None=MaterialFunctionCall!(
    Vector to Transform=WorldPosition().XYZ
    Z Vector=ComponentMask(None=@n7) {b=true, g=true, r=true}
    Center Location=ComponentMask(None=@n8) {b=true, g=true, r=true}
   ) {materialFunction=/Engine/Functions/Engine_MaterialFunctions02/Math/TransformToZVector}.Result) {b=true}
  n10 = Multiply(
   A=ConstantBiasScale(None=TextureSampleParameter2D() {parameterName=Displacement, samplerType=SAMPLERTYPE_LinearColor, texture=&127grey}.R) {bias=-0.5, scale=2}
   B=ScalarParameter() {parameterName=Displacement Height}
  )
  n11 = ScalarParameter() {parameterName=Min}
  n12 = Divide(
   A=Subtract(A=Subtract(A=@n9, B=@n10), B=@n11)
   B=Subtract(A=ScalarParameter() {defaultValue=100, parameterName=Max}, B=@n11) {desc="Length\n"}
  )
  Ceil(None=Subtract(A=Add(A=@n5, B=@n6), B=@n12))
  n13 = Power(Base=TextureSampleParameter2D(UVs=@n4) {parameterName=Initial Depth, samplerType=SAMPLERTYPE_LinearColor, texture=&127grey}.R) {constExponent=2.2}
  Clamp(None=Multiply(A=Distance(A=@n13, B=@n5)) {constB=4})
  Constant()
  MaterialFunctionCall!(UVW=@n12) {materialFunction=/Engine/ArtTools/RenderToTexture/MaterialFunctions/MarkRedOutside0-1Range}
 ]
}
