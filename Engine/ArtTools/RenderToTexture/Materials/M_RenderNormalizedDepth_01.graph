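# Unlit, two-sided masked render-to-texture material that, as its name indicates, renders normalized depth.
# BoundingSphere_LightTransform takes the MPC_R2T camera forward axis as light vector, with scale and local-space parameters, and its clamped W output passes through MarkRedOutside0-1Range.
# A Mask Channel dot product selects the mask texture channel, and Mask Offset expands or shrinks it against edge artifacts.
# A Push parameter scales the world vertex normal.
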Material /Engine/ArtTools/RenderToTexture/Materials/M_RenderNormalizedDepth_01
graph {
 properties {
  bUsedWithStaticLighting=true
  blendMode=BLEND_Masked
  numCustomizedUVs=3
  pixelDepthOffsetMode=PDOM_Legacy
  shadingModel=MSM_Unlit
  twoSided=true
  usedShadingModels=Unlit
 }
 nodes [
  n1 = TextureSampleParameter2D() {parameterName=Opacity Mask, texture=/Engine/EngineVolumetrics/Fogsheet/Materials/T_EV_BlankWhite_01}
  n2 = VectorParameter() {parameterName=Mask Channel, defaultValue=<rgba 0 0 0 1>}
  Add(
   A=DotProduct(A=AppendVector(A=@n1.RGB, B=@n1.RGB), B=AppendVector(A=@n2.RGB, B=@n2.RGB))
   B=ScalarParameter() {desc=Expand or shrink mask to tweak edge artifacts, parameterName=Mask Offset}
  )
  n3 = ScalarParameter() {defaultValue=1, parameterName=Scale XY}
  n4 = StaticBool() {value=true}
  n5 = CollectionParameter() {collection=&/MPC_R2T, parameterName=CameraForwardAxis}
  MaterialFunctionCall!(UVW=MaterialFunctionCall!(
    World Position
    Scale XY=AppendVector(A=@n3, B=@n3)
    Scale Z=ScalarParameter() {defaultValue=1, parameterName=Scale Z}
    Clamp=@n4
    Transform to Local space=StaticBoolParameter() {parameterName=Transform to Local Space}
    Custom Light Vector=@n4
    Light Vector=ComponentMask(None=@n5) {b=true, g=true, r=true}
   ) {materialFunction=/Engine/ArtTools/RenderToTexture/MaterialFunctions/BoundingSphere_LightTransform}.W) {materialFunction=/Engine/ArtTools/RenderToTexture/MaterialFunctions/MarkRedOutside0-1Range}
  Multiply(A=VertexNormalWS(), B=ScalarParameter() {parameterName=Push})
 ]
}
