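# Unlit preview material in the RenderToTexture art tools for viewing a texture tiled.
# Texture coordinates are multiplied by the Tiling scalar parameter and used as the UVs to sample TilingTexturePreview in linear color.
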Material /Engine/ArtTools/RenderToTexture/Materials/M_TilingPreviewMat_01
graph {
 properties {pixelDepthOffsetMode=PDOM_Legacy, shadingModel=MSM_Unlit, usedShadingModels=Unlit}
 nodes [TextureSample(UVs=Multiply(A=TextureCoordinate(), B=ScalarParameter() {defaultValue=3, parameterName=Tiling})) {samplerType=SAMPLERTYPE_LinearColor, texture=/Engine/ArtTools/RenderToTexture/Textures/TilingTexturePreview}]
}
